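# Level sequence for the HeatShineTorn emote's effects, animating an equip-prop skeletal mesh actor alongside three Niagara effects.
# The prop actor holds a fixed transform and plays one skeletal animation section.
# Niagara components bound as its children use system spawn tracks staggered in time: Appear starts with the animation, then Disappear, then Implosion_02, each section extending past the animation's end.
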LevelSequence /BRCosmetics/Animation/Game/MainPlayer/Emotes/HeatShineTorn/FX/FX_Emote_HeatShineTorn_Sequence
bindings [
 NS_Emote_HeatShineTorn_Appear {
  id=#1
  parent_id=#2
  possessed_class=NiagaraComponent
  tracks [{type=MovieSceneNiagaraSystemTrack, sections [{type=MovieSceneNiagaraSystemSpawnSection, end_frame=1200, start_frame=0}]}]
 }
 NS_Emote_HeatShineTorn_Implosion_02 {
  id=#3
  parent_id=#2
  possessed_class=NiagaraComponent
  tracks [{type=MovieSceneNiagaraSystemTrack, sections [{type=MovieSceneNiagaraSystemSpawnSection, end_frame=1591, start_frame=391}]}]
 }
 NS_Emote_HeatShineTorn_Disappear {
  id=#4
  parent_id=#2
  possessed_class=NiagaraComponent
  tracks [{type=MovieSceneNiagaraSystemTrack, sections [{type=MovieSceneNiagaraSystemSpawnSection, end_frame=1446, start_frame=246}]}]
 }
 Emote_HeatShineTorn_CMM_Equip_Prop_M {
  id=#2
  possessed_class=FortSkeletalMeshActor
  tracks [
   {
    type=MovieScene3DTransformTrack
    sections [{
      type=MovieScene3DTransformSection
      channels [
       Location.X {type=MovieSceneScriptingDoubleChannel, default=240}
       Location.Y {type=MovieSceneScriptingDoubleChannel, default=1424}
       Location.Z {type=MovieSceneScriptingDoubleChannel, default=1992}
       Rotation.X {type=MovieSceneScriptingDoubleChannel, default=0}
       Rotation.Y {type=MovieSceneScriptingDoubleChannel, default=0}
       Rotation.Z {type=MovieSceneScriptingDoubleChannel, default=0}
       Scale.X {type=MovieSceneScriptingDoubleChannel, default=1}
       Scale.Y {type=MovieSceneScriptingDoubleChannel, default=1}
       Scale.Z {type=MovieSceneScriptingDoubleChannel, default=1}
      ]
     }]
   }
   {type=MovieSceneSkeletalAnimationTrack, sections [{type=MovieSceneSkeletalAnimationSection, end_frame=484, start_frame=0}]}
  ]
 }
]
display_rate {denominator=1, numerator=120}
tick_resolution {denominator=1, numerator=24000}
tracks []
