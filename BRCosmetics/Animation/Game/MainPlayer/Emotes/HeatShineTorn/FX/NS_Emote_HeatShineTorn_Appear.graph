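# Niagara system for the appear effect of the HeatShineTorn emote, a timed burst of flares, glow and sparks around a board.
# Staggered loop delays sequence small flares sweeping inward along X, sparks, a final flare, body glow and mesh airflow.
# The board glow and light follow a socket reader, while airflow and sparks scale with time of day.
# BodyGlow sets visibility from front-end flags; LWE board emitters show no contents, with refraction disabled.
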NiagaraSystem /BRCosmetics/Animation/Game/MainPlayer/Emotes/HeatShineTorn/FX/NS_Emote_HeatShineTorn_Appear
dependencies {
 usedDataInterfaces [
  &NiagaraDataInterfaceCurve
  &NiagaraDataInterfaceVector2DCurve
  &NiagaraDataInterfaceColorCurve
  &NiagaraDataInterfaceSocketReader
 ]
 usedDynamicInputs [&MakeVector2D, &SimulationPosition]
 usedModules [
  &self:SystemSpawnScript.SystemScriptSource.SystemScriptGraph.NiagaraNodeAssignment_0.SetVariables_#1
  &SystemState
 ]
}
emitters {
 AirFlow {
  emitterUpdateScript [
   &EmitterState(Life Cycle Mode=::NewEnumerator1, Loop Behavior=::NewEnumerator1, Loop Delay=0.3, Loop Duration=0.15, UseLoopDelay=true)
   &SpawnBurst_Instantaneous(Spawn Count=3, Use Spawn Probability=true, UseLoopCountLimit=false)
  ]
  particleSpawnScript [
   &PartitionParticles(Number Of Partitions=3)
   /SetVariables(Particles.Random=&FloatFromCurve(FloatCurve=DataInterface() {propertyValues {Curve {keys [{tangentMode=RCTM_User, time=0, value=0}, {tangentMode=RCTM_User, time=1, value=1}]}}})) {isSetParameters=true}
   &InitializeParticle(
    Color=<rgba 2 2 2 1>
    Lifetime=0.5
    Mesh Scale=&MergeHelper_VectorSwitch(0=<0.75 0.75 1.5>, 1=<1 1 2>, 2=<1.25 1.25 2.5>, 3=?, Index=$Particles.PartitionParticles.Partition, Num=::NewEnumerator1)
    Mesh Scale Mode=::NewEnumerator4
    Position Offset=&MergeHelper_VectorSwitch(0=<-20 0 -50>, 1=<60 0 -40>, 2=<170 0 -30>, 3=?, Index=$Particles.PartitionParticles.Partition, Num=::NewEnumerator1)
    UsePositionOffset=true
    Write Material Random=true
   )
   &InitialMeshOrientation(
    Mesh Axis To Orient=<0 0 0>
    Mesh Orientation Mode=::NewEnumerator2
    Orientation Vector=<0 0 0>
    Rotation=<0 0.25 0>
    Use Rotation Vector=true
   )
   &ShapeLocation(Box Size=<400 0 0>, MaskSpawnGroup=false, Override Seed=false, Shape Primitive=::NewEnumerator2, Surface Only=false, UseOwnerScale=true) {enabled=false}
   &AddVelocity(Velocity=<100 0 20>, Velocity Speed Scale=2)
   &CameraOffset(Camera Offset Amount=-50)
  ]
  particleUpdateScript [
   &Drag(Drag=3, Use Linear Drag=true, Use Rotational Drag=true)
   &SolveForcesAndVelocity(Clamp Velocity=false, Limit Acceleration=false)
   &ParticleState()
   &DynamicMaterialParameters(
    Index 0 Param 1=&FloatFromCurve(FloatCurve=DataInterface() {propertyValues {Curve {keys [{interpMode=RCIM_Cubic, tangentMode=RCTM_User, time=0.5, value=0}, {interpMode=RCIM_Cubic, tangentMode=RCTM_User, time=1, value=1}]}}})
    Index 0 Param 2=&FNTimeOfDayScalar(Night Scale=0, Morning Scale=1, Evening Scale=1)
    Index 0 Param 3=1
    Module.Param0WriteEnabled=true
    Module.Param1WriteEnabled=true
    Module.Param2WriteEnabled=true
    Module.Param3WriteEnabled=true
   )
   &ScaleMeshSize(Scale Factor=&Multiply_VectorByFloat(
     Vector=&Multiply_VectorByFloat(
      Vector=<1 1 1>
      Float=&FloatFromCurve(
       FloatCurve=DataInterface() {propertyValues {Curve {keys [
           {interpMode=RCIM_Cubic, leaveTangent=0.3117, tangentMode=RCTM_Break, time=0, value=1}
           {arriveTangent=2.2576, tangentMode=RCTM_User, time=1.002, value=2}
          ]}}}
       Scale Curve=0.6
      )
     )
     Float=1
    ))
   &UpdateMeshOrientation(Rotation Rate=&RandomRangeFloat(Minimum=-0.5, Maximum=0.5), Rotation Vector=<0 0 1>)
   &ScaleColor(
    Scale Alpha=&FloatFromCurve(
     FloatCurve=DataInterface() {propertyValues {Curve {keys [
         {arriveTangent=3.7468, interpMode=RCIM_Cubic, leaveTangent=3.7468, tangentMode=RCTM_User, time=0, value=0}
         {arriveTangent=-2.0516, interpMode=RCIM_Cubic, leaveTangent=-2.0516, tangentMode=RCTM_User, time=0.3, value=0.5}
         {arriveTangent=0.0484, interpMode=RCIM_Cubic, leaveTangent=0.0484, tangentMode=RCTM_User, time=1, value=0}
        ]}}}
     Scale Curve=0.4
    )
    Scale RGB=&MakeVectorFromLinearColorRGB(Color=&FNTimeOfDayColor(Night Color=<rgba 0.5088 0.6003 0.9 1>, Morning Color=<rgba 0.7 0.7 0.7 1>, Day Color=<rgba 1 1 1 1>, Evening Color=<rgba 0.7 0.7 0.7 1>))
    ScaleA=true
    ScaleRGB=true
   )
  ]
  rendererClasses [&NiagaraMeshRendererProperties]
  renderers [{rendererClass=&NiagaraMeshRendererProperties, rendererIndex=0}]
 }
 Board_Glow_and_Light {
  emitterUpdateScript [
   &EmitterState(
    Inactive Response=::NewEnumerator1
    Life Cycle Mode=::NewEnumerator1
    Loop Behavior=::NewEnumerator1
    Loop Delay=0.25
    Loop Duration=2.8
    UseLoopDelay=true
   )
   &SpawnBurst_Instantaneous(Use Spawn Probability=false, UseLoopCountLimit=false) {module=SpawnBurst_Instantaneous001}
  ]
  particleSpawnScript [
   &InitializeParticle(Lifetime=0.5, Sprite Size=<80.2566 190.9906>, Sprite Size Mode=::NewEnumerator4, UsePositionOffset=false, Write Material Random=true)
   &DynamicMaterialParameters(
    Index 0 Param 1=2.4463
    Index 0 Param 2=90
    Index 0 Param 3=0.2089
    Index 0 Param 4=1
    Module.Index0Float0Write=true
    Module.Index0Float1Write=true
    Module.Index0Float2Write=true
    Module.Index0Float3Write=true
   )
   &CameraOffset(Camera Offset Amount=20)
  ]
  particleUpdateScript [
   &ParticleState()
   &SocketLocation(
    SamplePosOffset=true
    Sampled Position Offset=<-3.2595 0 0>
    Socket Reader=$System.Socket Reader
    Socket Sampling Mode=::NewEnumerator3
    VelocitySpeedLimit=false
   )
   &ScaleSpriteSize(
    Scale Sprite Size Mode=::NewEnumerator0
    Uniform Scale Factor=&MakeFloatFromVector(Vector=$Output.SocketLocation.SampledScale, Channel=::NewEnumerator0)
   )
   &ScaleColor(
    Scale Alpha=&FloatFromCurve(
     FloatCurve=DataInterface() {propertyValues {Curve {
        defaultValue=1
        keys [{interpMode=RCIM_Cubic, tangentMode=RCTM_User, time=0, value=1}, {interpMode=RCIM_Cubic, tangentMode=RCTM_User, time=1, value=0}]
       }}}
     Scale Curve=2
    )
    Scale RGB=&MakeVectorFromLinearColorRGB(Color=<rgba 0.3722 0.8551 1.2 0.0001>)
    ScaleA=true
    ScaleRGB=true
   ) {module=ScaleColor001}
   &SpriteFacingAndAlignment(Sprite Alignment=<1 0 0>, Sprite Facing=<0 0 1>, WriteAlignment=true, WriteFacing=true)
   &FN_Light_Attributes(
    Color=$Particles.Color
    Create Color Attribute=true
    Create Position Attribute=true
    Position=&AddVectorToPosition(Position=$Particles.Position, Vector=<0 0 20>)
    Scale Color=&Multiply_Float(
     A=3
     B=&FloatFromCurve(FloatCurve=DataInterface() {propertyValues {Curve {
         defaultValue=1
         keys [{interpMode=RCIM_Cubic, tangentMode=RCTM_User, time=0, value=1}, {interpMode=RCIM_Cubic, tangentMode=RCTM_User, time=1, value=0}]
        }}})
    )
   )
  ]
  rendererClasses [&NiagaraSpriteRendererProperties, &NiagaraLightRendererProperties]
  renderers [
   {rendererClass=&NiagaraSpriteRendererProperties}
   {rendererClass=&NiagaraLightRendererProperties, rendererIndex=1}
  ]
 }
 BodyGlow {
  emitterUpdateScript [
   &EmitterState(
    Inactive Response=::NewEnumerator1
    Life Cycle Mode=::NewEnumerator1
    Loop Behavior=::NewEnumerator1
    Loop Delay=0.26
    Loop Duration=0.4
    UseLoopDelay=true
   )
   &SpawnBurst_Instantaneous(Use Spawn Probability=false, UseLoopCountLimit=false)
  ]
  particleSpawnScript [
   /SetVariables(Particles.VisibilityTag=&MakeCustomIntFromBool(Bool=&Boolean_OrOperation(Bool1=$User.bIsFrontEnd, Bool2=$User.bIsFrontEndPreview), True INT=1, False INT=0)) {isSetParameters=true}
   &InitializeParticle(
    Color Mode=::NewEnumerator0
    Lifetime=0.16
    Mesh Scale=&FN_FrontEndVector3Modifier(Gameplay=<1 2 2.25>, Front End=<1 2 2.25>)
    Mesh Scale Mode=::NewEnumerator4
    Position Offset=<22 0 25>
    UsePositionOffset=true
    Write Material Random=true
   )
  ]
  particleUpdateScript [
   &ParticleState()
   &CameraOffset()
   /SetVariables(Particles.MeshOrientationFE:Quat4f=&LerpQuaternion(
     Quaternion A:Quat4f=$Particles.MeshOrientation
     Quaternion B:Quat4f=<0 1 0 1>
     Lerp Factor:NiagaraFloat=1
     Enable Spherical Linear Interpolation:NiagaraBool=false
    )) {isSetParameters=true}
   &ScaleColor(
    Scale Alpha=&FloatFromCurve(
     FloatCurve=DataInterface() {propertyValues {Curve {
        defaultValue=1
        keys [
         {arriveTangent=9.0725, interpMode=RCIM_Cubic, leaveTangent=9.0725, tangentMode=RCTM_User, time=0, value=0}
         {arriveTangent=0.2675, interpMode=RCIM_Cubic, leaveTangent=0.2675, tangentMode=RCTM_User, time=0.1408, value=0.9847}
         {arriveTangent=-0.0405, interpMode=RCIM_Cubic, leaveTangent=-0.0405, tangentMode=RCTM_User, time=1, value=0}
        ]
       }}}
     Scale Curve=2
    )
    Scale RGB=&VectorFromFloat(Value=1)
    ScaleA=true
    ScaleRGB=true
   )
  ]
  rendererClasses [&NiagaraMeshRendererProperties]
  renderers [
   {rendererClass=&NiagaraMeshRendererProperties, rendererIndex=0}
   {rendererClass=&NiagaraMeshRendererProperties, rendererIndex=1}
  ]
 }
 Final_Flare_Burst {
  emitterUpdateScript [
   &EmitterState(Life Cycle Mode=::NewEnumerator1, Loop Behavior=::NewEnumerator1, Loop Delay=0.26, Loop Duration=0.3, UseLoopDelay=true)
   &SpawnBurst_Instantaneous(Use Spawn Probability=false, UseLoopCountLimit=false) {module=SpawnBurst_Instantaneous001}
  ]
  particleSpawnScript [
   &InitializeParticle(
    AdjustAlpha=false
    AdjustHue=true
    AdjustSaturation=true
    AdjustValue=true
    Color=<rgba 0.2179 0.8284 1 1>
    Color Mode=::NewEnumerator3
    Hue Shift Range:Vector2f=<-0.01 0.01>
    Lifetime=0.2
    Position Offset=<90 0 -48.6837>
    Sprite Rotation Angle=0.5
    Sprite Rotation Mode=::NewEnumerator1
    Sprite Size Mode=::NewEnumerator3
    Uniform Sprite Size=500
    UsePositionOffset=true
    Write Material Random=true
   )
   &CameraOffset(Camera Offset Amount=50)
  ]
  particleUpdateScript [
   &ParticleState()
   &ScaleSpriteSize(Uniform Curve Sprite Scale=DataInterface() {propertyValues {Curve {
       defaultValue=1
       keys [
        {arriveTangent=9.0725, interpMode=RCIM_Cubic, leaveTangent=9.0725, tangentMode=RCTM_User, time=0, value=0}
        {arriveTangent=0.2676, interpMode=RCIM_Cubic, leaveTangent=0.2676, tangentMode=RCTM_User, time=0.1408, value=0.9847}
        {arriveTangent=-0.0405, interpMode=RCIM_Cubic, leaveTangent=-0.0405, tangentMode=RCTM_User, time=1, value=0}
       ]
      }}}) {module=ScaleSpriteSize001}
   &FN_Mesh_Attributes(
    Color=&MakeLinearColorFromVectorAndFloat(
     "Vector (RGB)"=&MakeVectorFromLinearColorRGB(Color=$Particles.Initial.Color)
     "Float (Alpha)"=&FloatFromCurve(
      FloatCurve=DataInterface() {propertyValues {Curve {
         defaultValue=1
         keys [
          {arriveTangent=9.0725, interpMode=RCIM_Cubic, leaveTangent=9.0725, tangentMode=RCTM_User, time=0, value=0}
          {arriveTangent=0.2675, interpMode=RCIM_Cubic, leaveTangent=0.2675, tangentMode=RCTM_User, time=0.1408, value=0.9847}
          {arriveTangent=-0.0405, interpMode=RCIM_Cubic, leaveTangent=-0.0405, tangentMode=RCTM_User, time=1, value=0}
         ]
        }}}
      Scale Curve=0.05
     )
    )
    Create Color Attribute=true
    Create Scale Attribute=true
    Scale=&VectorFromFloat(Value=2)
    Write to Module Namespace=true
   )
   /SetVariables(Particles.MeshGlow_DynPar:Vector4f=&MakeVector4(
     X=&FloatFromCurve(
      FloatCurve=DataInterface() {propertyValues {Curve {
         defaultValue=1
         keys [
          {arriveTangent=9.0725, interpMode=RCIM_Cubic, leaveTangent=9.0725, tangentMode=RCTM_User, time=0, value=0}
          {arriveTangent=0.2675, interpMode=RCIM_Cubic, leaveTangent=0.2675, tangentMode=RCTM_User, time=0.1408, value=0.9847}
          {arriveTangent=-0.0405, interpMode=RCIM_Cubic, leaveTangent=-0.0405, tangentMode=RCTM_User, time=1, value=0}
         ]
        }}}
      CurveIndex=$Emitter.NormalizedLoopAge
      Scale Curve=2
     )
     Y=0
     Z=2
     W=4.1611
    )) {isSetParameters=true}
   /SetVariables(Particles.MeshGlow_Pos:NiagaraPosition=&AddVectorToPosition(Position=$Particles.Position, Vector=<0 0 40>)) {isSetParameters=true}
   &FN_Light_Attributes(
    Color=$Particles.Color
    Create Color Attribute=true
    Create Position Attribute=true
    Position=&AddVectorToPosition(Position=$Particles.Position, Vector=<0 0 20>)
    Scale Color=&Multiply_Float(
     A=3
     B=&FloatFromCurve(FloatCurve=DataInterface() {propertyValues {Curve {
         defaultValue=1
         keys [{interpMode=RCIM_Cubic, tangentMode=RCTM_User, time=0, value=1}, {interpMode=RCIM_Cubic, tangentMode=RCTM_User, time=1, value=0}]
        }}})
    )
   )
  ]
  rendererClasses [
   &NiagaraSpriteRendererProperties
   &NiagaraMeshRendererProperties
   &NiagaraLightRendererProperties
  ]
  renderers [
   {rendererClass=&NiagaraSpriteRendererProperties}
   {rendererClass=&NiagaraMeshRendererProperties, rendererIndex=1}
   {rendererClass=&NiagaraLightRendererProperties, rendererIndex=2}
  ]
 }
 LWE_Board_Refraction {bEnabled=false, rendererClasses [], renderers []}
 LWE_Board_Trail_Green {rendererClasses [], renderers []}
 LWE_Board_Trail_Pink {rendererClasses [], renderers []}
 LWE_Board_Wave {rendererClasses [], renderers []}
 Small_Flare_Bursts {
  emitterUpdateScript [
   &EmitterState(Life Cycle Mode=::NewEnumerator1, Loop Behavior=::NewEnumerator1, Loop Delay=0.13, Loop Duration=0.16, UseLoopDelay=true)
   &SpawnRate(SpawnRate=30, Use Spawn Probability=false)
  ]
  particleSpawnScript [
   &InitializeParticle(
    AdjustAlpha=false
    AdjustHue=true
    AdjustSaturation=true
    AdjustValue=true
    Color=<rgba 0.2179 0.8284 1 1>
    Color Mode=::NewEnumerator3
    Hue Shift Range:Vector2f=<-0.01 0.01>
    Lifetime=0.1
    Mesh Scale Mode=::NewEnumerator3
    Mesh Uniform Scale=&FloatFromCurve(
     FloatCurve=DataInterface() {propertyValues {Curve {
        defaultValue=1
        keys [{interpMode=RCIM_Cubic, tangentMode=RCTM_User, time=0, value=0.5}, {interpMode=RCIM_Cubic, tangentMode=RCTM_User, time=1, value=1}]
       }}}
     CurveIndex=$Emitter.NormalizedLoopAge
    )
    Position Offset=&Lerp_Vector(A=<400 0 -40>, B=<100 0 -50>, Alpha=$Emitter.NormalizedLoopAge)
    Sprite Rotation Mode=::NewEnumerator2
    Sprite Size Mode=::NewEnumerator3
    Sprite UV Mode=::NewEnumerator3
    Uniform Sprite Size=&FloatFromCurve(
     FloatCurve=DataInterface() {propertyValues {Curve {
        defaultValue=1
        keys [{interpMode=RCIM_Cubic, tangentMode=RCTM_User, time=0, value=0.5}, {interpMode=RCIM_Cubic, tangentMode=RCTM_User, time=1, value=1}]
       }}}
     CurveIndex=$Emitter.NormalizedLoopAge
     Scale Curve=400
    )
    UsePositionOffset=true
    Write Material Random=true
   )
   &CameraOffset(Camera Offset Amount=20)
   /SetVariables(Particles.MeshGlow_DynPar:Vector4f=&MakeVector4(
     X=&FloatFromCurve(
      FloatCurve=DataInterface() {propertyValues {Curve {
         defaultValue=1
         keys [{interpMode=RCIM_Cubic, tangentMode=RCTM_User, time=0, value=0.5}, {interpMode=RCIM_Cubic, tangentMode=RCTM_User, time=1, value=1}]
        }}}
      CurveIndex=$Emitter.NormalizedLoopAge
     )
     Y=5
     Z=2
     W=3
    )) {isSetParameters=true}
  ]
  particleUpdateScript [
   &ParticleState()
   &ScaleSpriteSize(
    Uniform Curve Scale=&Waveform("Waveform [1]"=::NewEnumerator0, "[1] Phase Scale"=28.2298, "[1] Frequency"=1, "[1] Amplitude Min/Max"=<0.9 1.1>)
    Uniform Curve Sprite Scale=DataInterface() {propertyValues {Curve {
       defaultValue=0
       keys [
        {arriveTangent=0.1872, interpMode=RCIM_Cubic, leaveTangent=0.1872, tangentMode=RCTM_User, time=0, value=1}
        {interpMode=RCIM_Cubic, tangentMode=RCTM_User, time=1.0019, value=0.5}
       ]
      }}}
   ) {module=ScaleSpriteSize001}
   &SpriteRotationRate(Rotation Rate=&RandomRangeFloat(Minimum=-5, Maximum=5))
   &SolveForcesAndVelocity(Clamp Velocity=false, Limit Acceleration=false)
   &ScaleColor(Scale Alpha=0.5, ScaleA=true, ScaleRGB=false) {module=ScaleColor001}
   &FN_Sprite_Attributes(
    Color=&MakeLinearColorFromVectorAndFloat(
     "Vector (RGB)"=&MakeVectorFromLinearColorRGB(Color=<rgba 0.0097 0.0457 0.0573 2>)
     "Float (Alpha)"=&Multiply_Float(
      A=&MakeFloatFromLinearColor(LinearColor=$Particles.Color, Channel:ENiagaraFloatFromLinearColorOptions=::NewEnumerator3)
      B=&FloatFromCurve(FloatCurve=DataInterface() {propertyValues {Curve {
          defaultValue=1.0113
          keys [
           {arriveTangent=9.0725, interpMode=RCIM_Cubic, leaveTangent=9.0725, tangentMode=RCTM_User, time=0, value=0}
           {arriveTangent=-2.5249, interpMode=RCIM_Cubic, leaveTangent=-2.5249, tangentMode=RCTM_User, time=0.1408, value=0.9847}
           {arriveTangent=-0.0405, interpMode=RCIM_Cubic, leaveTangent=-0.0405, tangentMode=RCTM_User, time=1, value=0}
          ]
         }}})
     )
    )
    Create Color Attribute=true
    Create Sprite Rotation Attribute=true
    Create SpriteSize Attribute=true
    SpriteRotation:NiagaraFloat=&RandomRangeFloat(Minimum=-360, Maximum=360)
    SpriteSize=&Multiply_Vector2D_ByFloat(Float=2, Vector2D=$Particles.SpriteSize)
    Write to Module Namespace=true
   )
   &FN_Light_Attributes(
    Color=$Particles.Initial.Color
    Create Color Attribute=true
    Create Position Attribute=true
    Position=&AddVectorToPosition(Position=$Particles.Position, Vector=<0 0 -3.9163>)
    Scale Color=7.9589
   )
   /SetVariables(Particles.MeshGlow_Color:LinearColor=&MakeLinearColorFromVectorAndFloat(
     "Vector (RGB)"=&MakeVectorFromLinearColorRGB(Color=$Particles.Color)
     "Float (Alpha)"=&Multiply_Float(
      A=&MakeFloatFromLinearColor(LinearColor=$Particles.Color, Channel:ENiagaraFloatFromLinearColorOptions=::NewEnumerator3)
      B=0.2
     )
    )) {isSetParameters=true}
  ]
  rendererClasses [
   &NiagaraSpriteRendererProperties
   &NiagaraLightRendererProperties
   &NiagaraMeshRendererProperties
  ]
  renderers [
   {rendererClass=&NiagaraSpriteRendererProperties}
   {rendererClass=&NiagaraSpriteRendererProperties, rendererIndex=1}
   {rendererClass=&NiagaraLightRendererProperties, rendererIndex=2}
   {rendererClass=&NiagaraMeshRendererProperties, rendererIndex=3}
  ]
 }
 Sparks {
  emitterUpdateScript [
   &EmitterState(Life Cycle Mode=::NewEnumerator1, Loop Behavior=::NewEnumerator1, Loop Delay=0.16, Loop Duration=0.2, UseLoopDelay=true)
   &SpawnRate(SpawnRate=50, Use Spawn Probability=true)
  ]
  particleSpawnScript [
   &InitializeParticle(
    Color=<rgba 6.8064 25 21.9646 1>
    Lifetime=&UniformRangedFloat(Minimum=0.4, Maximum=0.6)
    Position Offset Coordinate Space=::Local
    Sprite Size=&Multiply_Vector2(A=<2 8>, B=&Vector2DFromFloat(Value=&Multiply_Float(A=1, B=2)))
    Sprite Size Mode=::NewEnumerator4
    UsePositionOffset=true
    Write Material Random=false
   ) {module=InitializeParticle001}
   &AddVelocity(Velocity=&RandomRangeVector(Minimum=<-600 0 0>, Maximum=<-800 0 0>), Velocity Speed Scale=5) {module=AddVelocity001}
   &BoxLocation(
    Box Offset=<150 0 -35>
    Box Size=&MakeVector(X=250, Y=50, Z=50)
    MaskSpawnGroup=false
    Module.Override Seed=false
    Rotate Box=false
    Surface Only=false
   )
  ]
  particleUpdateScript [
   &ParticleState()
   &Drag(Drag=20, Use Linear Drag=true, Use Rotational Drag=false)
   &ScaleColor(
    Scale Alpha=&Multiply_Float(
     A=&FloatFromCurve(FloatCurve=DataInterface() {propertyValues {Curve {keys [
          {arriveTangent=11.9433, interpMode=RCIM_Cubic, leaveTangent=11.9433, tangentMode=RCTM_User, time=0, value=0}
          <at 0.2 1 RCIM_Cubic>
          <at 1 0 RCIM_Cubic>
         ]}}})
     B=&Multiply_Float(A=1, B=&FNTimeOfDayScalar(Night Scale=0.5, Morning Scale=0.25, Evening Scale=0.25))
    )
    ScaleA=true
    ScaleRGB=true
   ) {module=ScaleColor002}
   &ScaleSpriteSize(Scale Factor=&Vector2DFromCurve(
     Vector2Curve=DataInterface() {propertyValues {
       XCurve {keys [<at 0 1>,<at 1 0>]}
       YCurve {keys [
         {arriveTangent=-11.1847, interpMode=RCIM_Cubic, leaveTangent=-11.1847, tangentMode=RCTM_User, time=0, value=4}
         {arriveTangent=-3.7, interpMode=RCIM_Cubic, leaveTangent=-3.7, time=0.3897, value=0.99}
         <at 1 0.3>
        ]}
      }}
     ScaleCurve=<1 1>
    ))
   &SolveForcesAndVelocity(Clamp Velocity=false, Limit Acceleration=false)
   &ScaleSpriteSizeBySpeed(Max Scale Factor=<1 4>, Min Scale Factor=<1 2>)
   &DynamicMaterialParameters(
    Index 0 Param 1=14.2084
    Index 0 Param 2=3.14
    Module.Index0Float0Write=true
    Module.Index0Float1Write=true
    Module.Index0Float2Write=true
    Module.Index0Float3Write=true
   )
   /SetVariables(
    Particles.GlowColor=&MakeLinearColorFromVectorAndFloat(
     "Vector (RGB)"=&MakeVectorFromLinearColorRGB(Color=<rgba 0.1696 0.6195 1 1>)
     "Float (Alpha)"=&FloatFromCurve(FloatCurve=DataInterface() {propertyValues {Curve {keys [<at 0.5 1>,<at 1 0>]}}}, Scale Curve=0.5)
    )
    Particles.GlowSize=&Multiply_Vector2D(A=$Particles.SpriteSize, B=<8 2>)
   ) {isSetParameters=true}
  ]
  rendererClasses [&NiagaraSpriteRendererProperties]
  renderers [
   {rendererClass=&NiagaraSpriteRendererProperties}
   {rendererClass=&NiagaraSpriteRendererProperties, rendererIndex=1}
  ]
 }
}
summary {
 systemName=NS_Emote_HeatShineTorn_Appear
 emitters [
  {emitterName=AirFlow, rendererClasses [&NiagaraMeshRendererProperties]}
  {emitterName=LWE_Board_Trail_Pink, rendererClasses []}
  {emitterName=LWE_Board_Wave, rendererClasses []}
  {emitterName=LWE_Board_Trail_Green, rendererClasses []}
  {
   emitterName=Board_Glow_and_Light
   rendererClasses [&NiagaraSpriteRendererProperties, &NiagaraLightRendererProperties]
  }
  {
   emitterName=Small_Flare_Bursts
   rendererClasses [
    &NiagaraSpriteRendererProperties
    &NiagaraLightRendererProperties
    &NiagaraMeshRendererProperties
   ]
  }
  {
   emitterName=Final_Flare_Burst
   rendererClasses [
    &NiagaraSpriteRendererProperties
    &NiagaraMeshRendererProperties
    &NiagaraLightRendererProperties
   ]
  }
  {emitterName=BodyGlow, rendererClasses [&NiagaraMeshRendererProperties]}
  {emitterName=Sparks, rendererClasses [&NiagaraSpriteRendererProperties]}
  {bEnabled=false, emitterName=LWE_Board_Refraction, rendererClasses []}
 ]
 userVariables [
  User.bIsFrontEnd {type=NiagaraBool}
  User.bIsFrontEndPreview {type=NiagaraBool}
 ]
}
