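# Niagara system for the disappear effect of the HeatShineTorn emote, glowing cyan to violet.
# Waves_Buildup spawns expanding mesh waves that intensify over the loop, while CenterWave bursts a delayed mesh with a flickering, fading light.
# Front_Flares and Speed_Sparks ramp up spawn rates for short-lived sprites.
# Sparks get curve-driven positions streaking along -X with radial spread, aligned to their frame-to-frame movement.
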NiagaraSystem /BRCosmetics/Animation/Game/MainPlayer/Emotes/HeatShineTorn/FX/NS_Emote_HeatShineTorn_Disappear
dependencies {
 usedDataInterfaces [
  &NiagaraDataInterfaceCurve
  &NiagaraDataInterfaceMeshRendererInfo
  &NiagaraDataInterfaceColorCurve
  &NiagaraDataInterfaceSocketReader
 ]
 usedDynamicInputs [&ColorFromCurve]
 usedModules [
  &self:SystemSpawnScript.SystemScriptSource.SystemScriptGraph.NiagaraNodeAssignment_0.SetVariables_#1
  &SystemState
 ]
}
emitters {
 CenterWave {
  emitterUpdateScript [
   &EmitterState(
    Inactive Response=::NewEnumerator1
    Life Cycle Mode=::NewEnumerator1
    Loop Behavior=::NewEnumerator1
    Loop Delay=0.15
    Loop Duration=1
    UseLoopDelay=true
   )
   &SpawnBurst_Instantaneous(Use Spawn Probability=false, UseLoopCountLimit=false)
  ]
  particleSpawnScript [
   &InitializeParticle(
    Lifetime=1.85
    Mesh Scale=&Multiply_VectorByFloat(Vector=<0.15 0.1 0.2>, Float=1.4)
    Mesh Scale Mode=::NewEnumerator4
    Position Offset=<26 0 0>
    UsePositionOffset=true
    Write Material Random=true
   )
   &InitialMeshOrientation(
    Mesh Axis To Orient=<0 0 1>
    Mesh Orientation Mode=::NewEnumerator2
    Orientation Vector=<0 0 1>
    Rotation=<0 0.25 0>
    Use Rotation Vector=true
   )
   &CameraOffset()
  ]
  particleUpdateScript [
   &ParticleState()
   &ScaleColor(
    Scale Alpha=&FloatFromCurve(
     FloatCurve=DataInterface() {propertyValues {Curve {
        defaultValue=1
        keys [
         {arriveTangent=56.6858, interpMode=RCIM_Cubic, leaveTangent=56.6858, tangentMode=RCTM_User, time=0, value=1}
         {interpMode=RCIM_Cubic, tangentMode=RCTM_User, time=0.2, value=5}
        ]
       }}}
     Scale Curve=0.5
    )
    Scale RGB=&MakeVectorFromLinearColorRGB(Color=&RandomRangeLinearColor(Minimum=<rgba 0.851 2.4255 4 1>, Maximum=<rgba 0.6452 0.2493 5 1>, Evaluation Type=::NewEnumerator1))
    ScaleA=true
    ScaleRGB=true
   ) {module=ScaleColor001}
   &ScaleMeshSize()
   &FN_Light_Attributes(
    Color=$Particles.Color
    Create Color Attribute=true
    Create Position Attribute=true
    Position=&AddVectorToPosition(Position=$Particles.Position, Vector=<0 -20 0>)
    Scale Color=&Multiply_Float(
     A=&Multiply_Float(
      A=1.2
      B=&Waveform(
       "Waveform [1]"=::NewEnumerator0
       "[1] Frequency"=&FloatFromCurve(FloatCurve=DataInterface() {propertyValues {Curve {
           defaultValue=1
           keys [
            {interpMode=RCIM_Cubic, tangentMode=RCTM_User, time=0, value=3}
            {arriveTangent=11.8166, interpMode=RCIM_Cubic, leaveTangent=11.8166, tangentMode=RCTM_User, time=0.6, value=8}
           ]
          }}})
       "[1] Amplitude Min/Max"=<0.5 1>
      )
     )
     B=&FloatFromCurve(FloatCurve=DataInterface() {propertyValues {Curve {
         defaultValue=1
         keys [{interpMode=RCIM_Cubic, tangentMode=RCTM_User, time=0, value=1}, {interpMode=RCIM_Cubic, tangentMode=RCTM_User, time=1, value=0}]
        }}})
    )
   )
  ]
  rendererClasses [&NiagaraMeshRendererProperties, &NiagaraLightRendererProperties]
  renderers [
   {rendererClass=&NiagaraMeshRendererProperties, rendererIndex=0}
   {rendererClass=&NiagaraLightRendererProperties, rendererIndex=1}
  ]
 }
 Front_Flares {
  emitterUpdateScript [
   &EmitterState(Life Cycle Mode=::NewEnumerator1, Loop Behavior=::NewEnumerator1, Loop Duration=1, UseLoopDelay=true)
   &SpawnRate(
    SpawnRate=&FloatFromCurve(
     FloatCurve=DataInterface() {propertyValues {Curve {
        defaultValue=1
        keys [
         {arriveTangent=0.0667, interpMode=RCIM_Cubic, leaveTangent=0.0667, tangentMode=RCTM_User, time=0, value=0.25}
         {arriveTangent=2.2645, tangentMode=RCTM_User, time=1, value=1}
        ]
       }}}
     CurveIndex=$Emitter.NormalizedLoopAge
     Scale Curve=20
    )
    Use Spawn Probability=false
   )
  ]
  particleSpawnScript [
   &InitializeParticle(
    Color Maximum=<rgba 0.8738 0.5896 4 1>
    Color Minimum=<rgba 0.851 2.9503 4 1>
    Color Mode=::NewEnumerator2
    Lifetime Max=0.6
    Lifetime Min=0.4
    Lifetime Mode=::NewEnumerator1
    Mesh Renderer Array Visibility Mode=::NewEnumerator1
    Mesh Renderer Info=DataInterface() {propertyValues {MeshRenderer=&self:Front_Flares_0.NiagaraMeshRendererProperties_0}}
    Mesh Renderer Mesh Index:NiagaraInt32=&MakeIntFromFloat(
     Float=&FloatFromCurve(
      FloatCurve=DataInterface() {propertyValues {Curve {defaultValue=1, keys [<at 0 0 RCIM_Cubic>, {arriveTangent=2.8878, tangentMode=RCTM_User, time=1, value=1}]}}}
      CurveIndex=&RandomRangeFloat(Minimum=0, Maximum=1)
     )
     Conversion Mode=::NewEnumerator1
    )
    Mesh Scale=&Lerp_Vector(
     A=<0.8 0.8 0.8>
     B=<1 1 1>
     Alpha=&FloatFromCurve(
      FloatCurve=DataInterface() {propertyValues {Curve {
         defaultValue=1
         keys [
          {arriveTangent=0.4663, interpMode=RCIM_Cubic, leaveTangent=0.4663, tangentMode=RCTM_User, time=0, value=0}
          {arriveTangent=2.2578, tangentMode=RCTM_User, time=1, value=1}
         ]
        }}}
      CurveIndex=&RandomRangeFloat(Minimum=0, Maximum=1)
     )
    )
    Mesh Scale Mode=::NewEnumerator4
    Position Offset=<50 0 0>
    Sprite Rotation Mode=::NewEnumerator2
    Sprite Size Mode=::NewEnumerator3
    Uniform Sprite Size=&FloatFromCurve(
     FloatCurve=DataInterface() {propertyValues {Curve {
        defaultValue=1
        keys [
         {arriveTangent=41.0327, interpMode=RCIM_Cubic, leaveTangent=41.0327, tangentMode=RCTM_User, time=0, value=50}
         {arriveTangent=820.3138, interpMode=RCIM_Cubic, leaveTangent=820.3138, tangentMode=RCTM_User, time=1, value=400}
        ]
       }}}
     CurveIndex=&RandomRangeFloat(Minimum=0, Maximum=1)
     Scale Curve=0.25
    )
    UsePositionOffset=true
    Write Material Random=true
   )
   &ShapeLocation(
    Cylinder Height=5
    Cylinder Radius=5
    MaskSpawnGroup=false
    Override Seed=false
    Shape Primitive=::NewEnumerator1
    Surface Only=false
    UseOwnerScale=true
   )
   &InitialMeshOrientation(
    Mesh Orientation Mode=::NewEnumerator2
    Rotation=&Lerp_Vector(
     A=<0 0 0>
     B=&RandomRangeVector(Minimum=<0.2 0 0>, Maximum=<-1.4 0 0>)
     Alpha=&MakeFloatFromVector(
      Vector=&Divide_Vector(
       A:Vector3f=&ConvertPositionToVector(Input Position=$Particles.Position, Conversion Operation:ENiagara_LWCConvertPosToVec=::NewEnumerator0)
       B:Vector3f=<1 200 1>
      )
      Channel=::NewEnumerator1
     )
    )
    Use Rotation Vector=true
   )
  ]
  particleUpdateScript [
   &ParticleState()
   &ScaleSpriteSize(Uniform Curve Sprite Scale=DataInterface() {propertyValues {Curve {
       defaultValue=0
       keys [
        {arriveTangent=4.1785, interpMode=RCIM_Cubic, leaveTangent=4.1785, tangentMode=RCTM_User, time=0, value=0}
        <at 0.1 1 RCIM_Cubic>
        <at 0.5 0 RCIM_Cubic>
       ]
      }}})
   &ScaleColor(
    Scale Alpha=&FloatFromCurve(
     FloatCurve=DataInterface() {propertyValues {Curve {
        defaultValue=1
        keys [
         {arriveTangent=9.0725, interpMode=RCIM_Cubic, leaveTangent=9.0725, tangentMode=RCTM_User, time=0, value=0}
         <at 0.1408 0.9847 RCIM_Cubic>
         {arriveTangent=-0.0405, interpMode=RCIM_Cubic, leaveTangent=-0.0405, tangentMode=RCTM_User, time=1, value=0}
        ]
       }}}
     Scale Curve=0.8
    )
    Scale RGB=&VectorFromFloat(Value=&FloatFromCurve(
      FloatCurve=DataInterface() {propertyValues {Curve {
         defaultValue=1
         keys [
          {arriveTangent=9.0725, interpMode=RCIM_Cubic, leaveTangent=9.0725, tangentMode=RCTM_User, time=0, value=0}
          <at 0.0916 0.9847 RCIM_Cubic>
          {arriveTangent=-0.0405, interpMode=RCIM_Cubic, leaveTangent=-0.0405, tangentMode=RCTM_User, time=1, value=0}
         ]
        }}}
      Scale Curve=2
     ))
    ScaleA=true
    ScaleRGB=true
   )
   /SetVariables(Particles.Color - Flare:LinearColor=&MultiplyLinearColor(
     Linear Color=$Particles.Initial.Color
     Scale Factor=&MakeVector4FromVectorAndFloat(
      "Vector (RGB)"=&VectorFromFloat(Value=&FloatFromCurve(
        FloatCurve=DataInterface() {propertyValues {Curve {
           defaultValue=1
           keys [
            {arriveTangent=0.556, interpMode=RCIM_Cubic, leaveTangent=0.556, tangentMode=RCTM_User, time=0, value=0}
            {arriveTangent=-0.2562, interpMode=RCIM_Cubic, leaveTangent=-0.2562, tangentMode=RCTM_User, time=0.1016, value=0.8685}
            {arriveTangent=-0.0405, interpMode=RCIM_Cubic, leaveTangent=-0.0405, tangentMode=RCTM_User, time=0.4, value=0}
           ]
          }}}
        Scale Curve=&RandomRangeFloat(Minimum=1, Maximum=2)
       ))
      "Float (Alpha)"=2
     )
    )) {isSetParameters=true}
  ]
  rendererClasses [&NiagaraSpriteRendererProperties]
  renderers [{rendererClass=&NiagaraSpriteRendererProperties}]
 }
 Speed_Sparks {
  emitterUpdateScript [
   &EmitterState(Life Cycle Mode=::NewEnumerator1, Loop Behavior=::NewEnumerator1, Loop Delay=0.05, Loop Duration=1, UseLoopDelay=true)
   &SpawnRate(
    SpawnRate=&FloatFromCurve(
     FloatCurve=DataInterface() {propertyValues {Curve {defaultValue=1, keys [<at 0 0.25 RCIM_Cubic>, {arriveTangent=1.7266, tangentMode=RCTM_User, time=1, value=1}]}}}
     CurveIndex=$Emitter.NormalizedLoopAge
     Scale Curve=100
    )
    Use Spawn Probability=false
   )
  ]
  particleSpawnScript [&InitializeParticle(
    Color Maximum=<rgba 1.1747 0.5896 4 1>
    Color Minimum=<rgba 0.851 2.9503 4 1>
    Color Mode=::NewEnumerator2
    Lifetime Max=0.6
    Lifetime Min=0.4
    Lifetime Mode=::NewEnumerator1
    Position Offset=<50 0 0>
    Sprite Rotation Mode=::NewEnumerator2
    Sprite Size=&Multiply_Vector2D_ByFloat(Float=&RandomRangeFloat(Minimum=2, Maximum=5), Vector2D=<3 2>)
    Sprite Size Mode=::NewEnumerator4
    UsePositionOffset=true
    Write Material Random=true
   )]
  particleUpdateScript [
   &SolveForcesAndVelocity(Clamp Velocity=false, Limit Acceleration=false)
   &ParticleState()
   &ScaleSpriteSizeBySpeed(Max Scale Factor=<1 5>, Min Scale Factor=<0.2 1>)
   /SetVariables(Particles.Position=&AddVectorToPosition(
     Position=$Particles.Initial.Position
     Vector=&Multiply_Vector(
      A=&Add_Vector(
       A=&Lerp_Vector(
        A=<0 0 0>
        B=<-250 0 0>
        Alpha=&FloatFromCurve(FloatCurve=DataInterface() {propertyValues {Curve {keys [
             {arriveTangent=0.2266, interpMode=RCIM_Cubic, leaveTangent=0.2266, tangentMode=RCTM_User, time=0, value=0}
             {arriveTangent=0.8833, interpMode=RCIM_Cubic, leaveTangent=0.8833, tangentMode=RCTM_User, time=1, value=1}
            ]}}})
       )
       B=&Multiply_VectorByFloat(
        Vector=&RotateVectorByQuaternion(
         Vector To Rotate=<0 0 1>
         Quaternion=&MakeQuaternion(
          Quaternion From=0
          Axis=<1 0 0>
          Angle=&RandomRangeFloat(Minimum=0, Maximum=1)
          From=?
          To=?
          Angle Type=::NewEnumerator1
          Normalize Vectors=false
         )
        )
        Float=&Lerp_Float(
         A=0
         B=25
         Alpha=&FloatFromCurve(FloatCurve=DataInterface() {propertyValues {Curve {keys [
              {arriveTangent=3.7638, interpMode=RCIM_Cubic, leaveTangent=3.7638, tangentMode=RCTM_User, time=0, value=0}
              {interpMode=RCIM_Cubic, tangentMode=RCTM_User, time=1, value=1}
             ]}}})
        )
       )
      )
      B=<1 1 1>
     )
    )) {isSetParameters=true}
   &ScaleColor(
    Scale Alpha=&FloatFromCurve(
     FloatCurve=DataInterface() {propertyValues {Curve {
        defaultValue=1
        keys [
         {arriveTangent=9.0725, interpMode=RCIM_Cubic, leaveTangent=9.0725, tangentMode=RCTM_User, time=0, value=0}
         <at 0.1408 0.9847 RCIM_Cubic>
         {arriveTangent=-0.0405, interpMode=RCIM_Cubic, leaveTangent=-0.0405, tangentMode=RCTM_User, time=1, value=0}
        ]
       }}}
     Scale Curve=0.8
    )
    Scale RGB=&VectorFromFloat(Value=&FloatFromCurve(
      FloatCurve=DataInterface() {propertyValues {Curve {
         defaultValue=1
         keys [
          {arriveTangent=9.0725, interpMode=RCIM_Cubic, leaveTangent=9.0725, tangentMode=RCTM_User, time=0, value=0}
          <at 0.0916 0.9847 RCIM_Cubic>
          {arriveTangent=-0.0405, interpMode=RCIM_Cubic, leaveTangent=-0.0405, tangentMode=RCTM_User, time=1, value=0}
         ]
        }}}
      Scale Curve=2
     ))
    ScaleA=true
    ScaleRGB=true
   )
   &SpriteFacingAndAlignment(
    Sprite Alignment=&NormalizeVector(
     Vector To Normalize=&Subtract_Position(A=$Particles.Position, B=$Particles.Previous.Position)
     Safe Normalize=true
     Safe Fallback Threshold=0
     Safe Fallback Vector=<0 0 0.001>
    )
    WriteAlignment=true
    WriteFacing=false
   )
   &FN_Sprite_Attributes(
    Color=&MultiplyLinearColor(
     Linear Color=$Particles.Color
     Scale Factor=&MakeVector4_FromLinearColor(LinearColor=<rgba 0.1927 0.1927 0.1927 0.01>)
    )
    Create Color Attribute=true
    Create SpriteSize Attribute=true
    SpriteSize=&Multiply_Vector2D(A=$Particles.SpriteSize, B=<47.6186 5.1085>)
    Write to Module Namespace=true
   )
   &DynamicMaterialParameters(
    Index 0 Param 1=5
    Index 0 Param 2=31.4447
    Index 0 Param 3=1
    Index 0 Param 4=-0.0215
    Module.Index0Float0Write=true
    Module.Index0Float1Write=true
    Module.Index0Float2Write=true
    Module.Index0Float3Write=true
   )
  ]
  rendererClasses [&NiagaraSpriteRendererProperties]
  renderers [
   {rendererClass=&NiagaraSpriteRendererProperties}
   {rendererClass=&NiagaraSpriteRendererProperties, rendererIndex=1}
  ]
 }
 Waves_Buildup {
  emitterUpdateScript [
   &EmitterState(Life Cycle Mode=::NewEnumerator1, Loop Behavior=::NewEnumerator1, Loop Delay=0.1, Loop Duration=1, UseLoopDelay=true)
   &SpawnRate(
    SpawnRate=&FloatFromCurve(
     FloatCurve=DataInterface() {propertyValues {Curve {
        defaultValue=1
        keys [
         {interpMode=RCIM_Cubic, tangentMode=RCTM_User, time=0, value=4}
         {arriveTangent=14.293, interpMode=RCIM_Cubic, leaveTangent=14.293, tangentMode=RCTM_User, time=1, value=8}
        ]
       }}}
     CurveIndex=$Emitter.NormalizedLoopAge
     Scale Curve=1.25
    )
    Use Spawn Probability=false
   )
  ]
  particleSpawnScript [
   &InitializeParticle(
    Color Maximum=<rgba 0.8738 0.5896 4 1>
    Color Minimum=<rgba 0.851 2.9503 4 1>
    Color Mode=::NewEnumerator2
    Lifetime=0.4
    Mesh Scale=<0.35 0.2 0.2>
    Mesh Scale Mode=::NewEnumerator4
    Position Offset=<45 0 0>
    UsePositionOffset=true
    Write Material Random=true
   )
   &CameraOffset()
  ]
  particleUpdateScript [
   &ParticleState()
   &DynamicMaterialParameters(
    Index 0 Param 1=&FloatFromCurve(FloatCurve=DataInterface() {propertyValues {Curve {keys [{arriveTangent=-2.1015, interpMode=RCIM_Cubic, leaveTangent=-2.1015, tangentMode=RCTM_User, time=0, value=0.5}, <at 1 -0.15>]}}})
    Index 0 Param 2=&FloatFromCurve(FloatCurve=DataInterface() {propertyValues {Curve {keys [{arriveTangent=-0.3464, interpMode=RCIM_Cubic, leaveTangent=-0.3464, tangentMode=RCTM_User, time=0, value=1}, <at 1 0.9>]}}})
    Index 0 Param 3=&FloatFromCurve(FloatCurve=DataInterface() {propertyValues {Curve {keys [<at 0 0>,<at 0.15 1>,<at 1 1.2>]}}})
    Module.Param0WriteEnabled=true
    Module.Param1WriteEnabled=true
    Module.Param2WriteEnabled=true
    Module.Param3WriteEnabled=false
   )
   &ScaleColor(
    Scale Alpha=&Multiply_Float(
     A=&FloatFromCurve(FloatCurve=DataInterface() {propertyValues {Curve {keys [<at 0 1>,<at 1 0>]}}}, Scale Curve=0.1)
     B=&FloatFromCurve(
      FloatCurve=DataInterface() {propertyValues {Curve {
         defaultValue=1
         keys [
          {interpMode=RCIM_Cubic, tangentMode=RCTM_User, time=0, value=1}
          {arriveTangent=1.8836, interpMode=RCIM_Cubic, leaveTangent=1.8836, tangentMode=RCTM_User, time=1, value=2}
         ]
        }}}
      CurveIndex=$Emitter.NormalizedLoopAge
     )
    )
    ScaleA=true
    ScaleRGB=true
   ) {module=ScaleColor001}
   &ScaleMeshSize(Scale Factor=&Multiply_Vector(
     A=&VectorFromFloat(Value=&FloatFromCurve(
       FloatCurve=DataInterface() {propertyValues {Curve {keys [<at 0 0>,<at 0.1 1.4>,<at 1 3>]}}}
       Scale Curve=&Multiply_Float(A=1, B=1)
      ))
     B=&MakeVector(
      X=&FloatFromCurve(
       FloatCurve=DataInterface() {propertyValues {Curve {
          defaultValue=1
          keys [
           {interpMode=RCIM_Cubic, tangentMode=RCTM_User, time=0, value=1}
           {arriveTangent=3.5471, interpMode=RCIM_Cubic, leaveTangent=3.5471, tangentMode=RCTM_User, time=1, value=2.5}
          ]
         }}}
       CurveIndex=$Emitter.NormalizedLoopAge
      )
      Y=1
      Z=1
     )
    ))
  ]
  rendererClasses [&NiagaraMeshRendererProperties]
  renderers [{rendererClass=&NiagaraMeshRendererProperties, rendererIndex=0}]
 }
}
summary {
 systemName=NS_Emote_HeatShineTorn_Disappear
 emitters [
  {emitterName=Waves_Buildup, rendererClasses [&NiagaraMeshRendererProperties]}
  {emitterName=CenterWave, rendererClasses [&NiagaraMeshRendererProperties, &NiagaraLightRendererProperties]}
  {emitterName=Front_Flares, rendererClasses [&NiagaraSpriteRendererProperties]}
  {emitterName=Speed_Sparks, rendererClasses [&NiagaraSpriteRendererProperties]}
 ]
 userVariables [
  User.bIsFrontEnd {type=NiagaraBool}
  User.bIsFrontEndPreview {type=NiagaraBool}
 ]
}
