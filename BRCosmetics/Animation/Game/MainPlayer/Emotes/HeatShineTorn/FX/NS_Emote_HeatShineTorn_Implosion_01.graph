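# Niagara system for the implosion effect of the HeatShineTorn player emote.
# A refraction flare with sprite and light joins burst-spawned smoke, sparks, streaks and mesh waves, plus brief smoky trails.
# Smoke and sparks launch outward, then swirl back via vortex and rising point attraction, dying inside a growing kill radius.
# Spawn groups vary flare and wave layers; front-end modifiers adjust size and brightness, and time of day tints smoke and waves.
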NiagaraSystem /BRCosmetics/Animation/Game/MainPlayer/Emotes/HeatShineTorn/FX/NS_Emote_HeatShineTorn_Implosion_01
dependencies {
 usedDataInterfaces [
  &NiagaraDataInterfaceCurve
  &NiagaraDataInterfaceVector2DCurve
  &NiagaraDataInterfaceArrayFloat3
  &NiagaraDataInterfaceColorCurve
  &NiagaraDataInterfaceArrayFloat
  &NiagaraDataInterfaceVectorCurve
  &NiagaraDataInterfacePlatformSet
 ]
 usedDynamicInputs [&SimulationPosition, &ReturnNormalizedExecIndex]
 usedModules [
  &self:SystemSpawnScript.SystemScriptSource.SystemScriptGraph.NiagaraNodeAssignment_0.SetVariables_#1
  &SystemState
 ]
}
emitters {
 LWE_Refract {rendererClasses [], renderers []}
 Refraction_Flare {
  emitterUpdateScript [
   &EmitterState(Life Cycle Mode=::NewEnumerator1, Loop Behavior=::NewEnumerator1, Loop Duration=0.4, UseLoopDelay=false)
   &SpawnBurst_Instantaneous(Spawn Count=1, Use Spawn Probability=false)
   &SpawnBurst_Instantaneous(Spawn Count=1, Spawn Group=1, Use Spawn Probability=false) {module=SpawnBurst_Instantaneous001}
  ]
  particleSpawnScript [
   &InitializeParticle(
    Color=<rgba 5 1.4827 0.5 1>
    Lifetime=0.62
    Mass Mode=::NewEnumerator1
    Mesh Scale Mode=::NewEnumerator4
    Sprite Size Mode=::NewEnumerator4
    UsePositionOffset=true
    Write Material Random=true
   ) {module=InitializeParticle001}
   /SetVariables(Particles.VisibilityTag=&SpawnGroup()) {isSetParameters=true}
  ]
  particleUpdateScript [
   &ParticleState()
   &ScaleMeshSize(Scale Factor=&VectorFromFloat(Value=&MaskFloatBySpawnGroup(
      Masked Float=&FloatFromCurve(FloatCurve=DataInterface() {propertyValues {Curve {keys [{arriveTangent=5.2157, interpMode=RCIM_Cubic, leaveTangent=5.2157, tangentMode=RCTM_User, time=0, value=0.3}, <at 0.4 1 RCIM_Cubic>]}}})
      Passthrough Float=&FloatFromCurve(FloatCurve=DataInterface() {propertyValues {Curve {keys [
           {interpMode=RCIM_Cubic, tangentMode=RCTM_Break, time=0, value=0.2}
           {arriveTangent=2.9365, interpMode=RCIM_Cubic, leaveTangent=2.9365, tangentMode=RCTM_User, time=0.1, value=1}
           {interpMode=RCIM_Cubic, tangentMode=RCTM_Break, time=1, value=2}
          ]}}})
     )))
   &ScaleColor(
    Scale Alpha=&MaskFloatBySpawnGroup(
     Masked Float=0
     Passthrough Float=&FloatFromCurve(FloatCurve=DataInterface() {propertyValues {Curve {keys [
          {interpMode=RCIM_Cubic, tangentMode=RCTM_User, time=0, value=1}
          {arriveTangent=-2.6886, interpMode=RCIM_Cubic, leaveTangent=-2.6886, tangentMode=RCTM_User, time=1, value=0}
         ]}}})
    )
    Scale RGB=&MaskVectorBySpawnGroup(
     Masked Vector=&VectorFromFloat(Value=&FloatFromCurve(FloatCurve=DataInterface() {propertyValues {Curve {keys [<at 0 0.7>,<at 0.7729 0.7>,<at 1 0>]}}}, Scale Curve=0.2))
     Passthrough Vector=&Multiply_Vector(
      A=&MakeVectorFromLinearColorRGB(Color=&ColorFromCurve(ColorCurve=DataInterface() {propertyValues {
          AlphaCurve {keys [<at 0 1>,<at 0.4696 1>,<at 1 0>]}
          BlueCurve {keys [<at 0 0.05>,<at 0.2681 0>,<at 0.4696 0.0698>,<at 1 0.037>]}
          GreenCurve {keys [<at 0 0.5707>,<at 0.2681 0.3605>,<at 0.4696 0.1719>,<at 1 0.0811>]}
          RedCurve {keys [<at 0 1>,<at 0.2681 1>,<at 0.4696 0.6979>,<at 1 0.3698>]}
         }}))
      B=&VectorFromFloat(Value=&FloatFromCurve(FloatCurve=DataInterface() {propertyValues {Curve {keys [
            {interpMode=RCIM_Cubic, tangentMode=RCTM_Break, time=0, value=10}
            {arriveTangent=-6.6072, interpMode=RCIM_Cubic, leaveTangent=-6.6072, tangentMode=RCTM_User, time=0.15, value=3}
            {interpMode=RCIM_Cubic, tangentMode=RCTM_Break, time=1, value=1}
           ]}}}))
     )
    )
    ScaleA=true
    ScaleRGB=true
   )
   &FN_Mesh_Attributes(
    Color=&MultiplyLinearColor(Linear Color=<rgba 1 0.4513 0 0.1652>, Scale Factor=<1 1 1 1>)
    Create Color Attribute=true
    Create Scale Attribute=true
    Scale=&VectorFromFloat(Value=&Multiply_Float(
      A=5
      B=&Multiply_Float(
       A=0.2
       B=&FloatFromCurve(FloatCurve=DataInterface() {propertyValues {Curve {
           defaultValue=0.1
           keys [
            {arriveTangent=9.0725, interpMode=RCIM_Cubic, leaveTangent=9.0725, tangentMode=RCTM_User, time=0, value=0}
            {arriveTangent=-2.5249, interpMode=RCIM_Cubic, leaveTangent=-2.5249, tangentMode=RCTM_User, time=0.1408, value=0.9847}
            {arriveTangent=-0.0405, interpMode=RCIM_Cubic, leaveTangent=-0.0405, tangentMode=RCTM_User, time=1, value=0}
           ]
          }}})
      )
     ))
    Write to Module Namespace=true
   )
   /SetVariables(Particles.RefracSphereDynMat:Vector4f=&MakeVector4(
     X=&Lerp_Float(
      A=1
      B=&FloatFromCurve(FloatCurve=DataInterface() {propertyValues {Curve {keys [<at 0 1>,<at 0.3172 1>,<at 1 1>]}}}, Scale Curve=0)
      Alpha=&FN_PickFloatDependOnPlatform(TargetPlatformActive=1, TargetPlatformNoActive=0)
     )
     Y=&FloatFromCurve(FloatCurve=DataInterface() {propertyValues {Curve {keys [<at 0 15>,<at 1 0>]}}})
     Z=0
     W=0
    )) {isSetParameters=true}
   &FN_Sprite_Attributes(
    Camera Offset:NiagaraFloat=&FN_FrontEndFloatModifier(Gameplay=25, Front End=0)
    Color=&MultiplyLinearColor(
     Linear Color=$Particles.Color
     Scale Factor=&MakeVector4FromVectorAndFloat(
      "Vector (RGB)"=&VectorFromFloat(Value=5)
      "Float (Alpha)"=&Waveform("Waveform [1]"=::NewEnumerator0, "[1] Frequency"=10, "[1] Amplitude Min/Max"=<0.2 1>)
     )
    )
    Create Camera Offset Attribute=true
    Create Color Attribute=true
    Create SpriteSize Attribute=true
    SpriteSize=&Multiply_Vector2D_ByFloat(
     Float=&Multiply_Float(
      A=&Multiply_Float(
       A=&FN_FrontEndFloatModifier(Gameplay=1500, Front End=1000)
       B=&FloatFromCurve(FloatCurve=DataInterface() {propertyValues {Curve {
           defaultValue=0
           keys [
            {arriveTangent=9.0725, interpMode=RCIM_Cubic, leaveTangent=9.0725, tangentMode=RCTM_User, time=0, value=0}
            {arriveTangent=-1.014, interpMode=RCIM_Cubic, leaveTangent=-1.014, tangentMode=RCTM_User, time=0.1408, value=0.9847}
            {arriveTangent=-2.261, interpMode=RCIM_Cubic, leaveTangent=-2.261, tangentMode=RCTM_User, time=1, value=0}
           ]
          }}})
      )
      B=0.2
     )
     Vector2D=<1 1>
    )
    Write to Module Namespace=true
   )
   &FN_Light_Attributes(
    Color=<rgba 10 3.6259 0 1>
    Create Color Attribute=true
    Create Exponent Attribute=true
    Create Position Attribute=true
    Create Radius Attribute=true
    Exponent=1.5
    Position=&AddVectorToPosition(
     Position=/Niagara/DynamicInputs/Helpers/SimulationPosition()
     Vector=&TransformVector(Vector=<0 0 0>, Destination Space=::World)
    )
    Radius=&Multiply_Float(A=130, B=&FN_FrontEndFloatModifier(Gameplay=1, Front End=1))
    Scale Color=&FloatFromCurve(
     FloatCurve=DataInterface() {propertyValues {Curve {keys [
         {interpMode=RCIM_Cubic, tangentMode=RCTM_User, time=0, value=1}
         {arriveTangent=-0.5426, interpMode=RCIM_Cubic, leaveTangent=-0.5426, tangentMode=RCTM_User, time=1, value=0}
        ]}}}
     Scale Curve=10
    )
   )
   &DynamicMaterialParameters(
    Index 0 Param 1=0.3
    Module.Index0Float0Write=true
    Module.Index0Float1Write=true
    Module.Index0Float2Write=true
    Module.Index0Float3Write=true
   )
  ]
  rendererClasses [
   &NiagaraMeshRendererProperties
   &NiagaraSpriteRendererProperties
   &NiagaraLightRendererProperties
  ]
  renderers [
   {rendererClass=&NiagaraMeshRendererProperties, rendererIndex=0}
   {rendererClass=&NiagaraSpriteRendererProperties, rendererIndex=1}
   {rendererClass=&NiagaraLightRendererProperties, rendererIndex=2}
  ]
 }
 Smoke {
  emitterSpawnScript [
   /SetVariables(
    Emitter.Global Color:LinearColor=<rgba 4 0.349 0 1>
    Emitter.Global Scale:NiagaraFloat=&Multiply_Float(A=0.15, B=&FN_FrontEndFloatModifier(Gameplay=1, Front End=1))
   ) {isSetParameters=true}
   /SetVariables(
    Emitter.Color2:LinearColor=&MultiplyLinearColor(Linear Color=$Emitter.Global Color, Scale Factor=<2 1 1 1>)
    Emitter.Depth Fade:NiagaraFloat=&Multiply_Float(A=125, B=$Emitter.Global Scale)
    Emitter.Dissolve Noise Power=1.1
    Emitter.Dissolve Noise Scale:LinearColor=&MultiplyLinearColor(
     Linear Color=/Niagara/DynamicInputs/TypeConversions/MakeLinearColorFromFloat(Value:NiagaraFloat=&FloatFromCurve(FloatCurve=DataInterface() {propertyValues {Curve {keys [<at 0 2>,<at 1 1>,<at 2 0.5>]}}}, CurveIndex=$Emitter.Global Scale))
     Scale Factor=<0.0008 0.0008 0.0009 1>
    )
    Emitter.Near Camera Fade:NiagaraFloat=&Multiply_Float(A=400, B=$Emitter.Global Scale)
   ) {isSetParameters=true}
   /SetVariables(Emitter.Dissolve Scale Adjsutment:NiagaraFloat=&FloatFromCurve(
     FloatCurve=DataInterface() {propertyValues {Curve {keys [<at 0 3>,<at 1 1>,<at 5 0.5>,<at 20 0.4>,<at 100 0.4>]}}}
     CurveIndex=$Emitter.Global Scale
     Scale Curve=$Emitter.Global Scale
    )) {isSetParameters=true}
  ]
  emitterUpdateScript [
   &EmitterState(Life Cycle Mode=::NewEnumerator1, Loop Behavior=::NewEnumerator1, Loop Duration=0.627, UseLoopDelay=false)
   &SpawnBurst_Instantaneous(Spawn Group=1, Spawn Time=0.01, Use Spawn Probability=false, UseLoopCountLimit=false) {enabled=false, module=SpawnBurst_Instantaneous002}
   &SpawnBurst_Instantaneous(Spawn Count=8, Use Spawn Probability=false, UseLoopCountLimit=false) {module=SpawnBurst_Instantaneous001}
  ]
  particleSpawnScript [
   &InitializeParticle(
    Color=&MaskLinearColorBySpawnGroup(Masked Linear Color=<rgba 0.9048 0.9062 0.9048 2>, Passthrough Linear Color=<rgba 0.4323 0.4252 0.4161 1.7>, Spawn Group Mask=1)
    Lifetime=&MaskFloatBySpawnGroup(
     Masked Float=1
     Passthrough Float=&SelectFloatFromArray(
      Array Sampling Mode=::NewEnumerator0
      Direct Array Index=?
      Float Selection Array=DataInterface() {propertyValues {FloatData [1,1.2]}}
      Random Seed=0
     )
     Spawn Group Mask=1
    )
    Mesh Scale Mode=::NewEnumerator3
    Mesh Uniform Scale=&RandomRangeFloat(Minimum=4, Maximum=5)
    UsePositionOffset=false
    Write Material Random=true
   )
   &ShapeLocation(MaskSpawnGroup=false, Offset Mode=::NewEnumerator0, Override Seed=false, Sphere Radius=1, UseOwnerScale=true)
   &AddVelocity(
    Module.Use Falloff Distance=false
    Origin Offset=&Multiply_VectorByFloat(Vector=&RandomRangeVector(Minimum=<0 0 0>, Maximum=<0 0 -5>), Float=$Emitter.Global Scale)
    Override Seed=false
    Velocity Mode=::NewEnumerator1
    Velocity Speed=&Multiply_Float(
     A=&MaskFloatBySpawnGroup(Masked Float=1400, Passthrough Float=&RandomRangeFloat(Minimum=900, Maximum=1200), Spawn Group Mask=1)
     B=$Emitter.Global Scale
    )
   )
   &CameraOffset(Camera Offset Amount=&Multiply_Float(A=&MaskFloatBySpawnGroup(Masked Float=70, Passthrough Float=25, Spawn Group Mask=1), B=$Emitter.Global Scale))
   /SetVariables(Particles.Scale=&Multiply_VectorByFloat(Vector=$Particles.Scale, Float=$Emitter.Global Scale)) {isSetParameters=true}
  ]
  particleUpdateScript [
   &ParticleState()
   &GravityForce(Gravity=&Multiply_VectorByFloat(
     Vector=&MaskVectorBySpawnGroup(
      Masked Vector=<0 0 500>
      Passthrough Vector=&RandomRangeVector(Minimum=<0 0 400>, Maximum=<1 1 600>)
      Spawn Group Mask=1
     )
     Float=$Emitter.Global Scale
    ))
   &Drag(Drag=&RandomRangeFloat(Minimum=4, Maximum=5), Use Linear Drag=true, Use Rotational Drag=false)
   &DynamicMaterialParameters(
    Index 0 Param 1=&Multiply_Float(
     A=&MaskFloatBySpawnGroup(
      Masked Float=-150
      Passthrough Float=&RandomRangeFloat(Minimum=-90, Maximum=-120)
      Spawn Group Mask=&MakeIntFromFloat(Float=1)
     )
     B=$Emitter.Global Scale
    )
    Index 0 Param 2=&FloatFromCurve(
     FloatCurve=DataInterface() {propertyValues {Curve {keys [
         {tangentMode=RCTM_User, time=0, value=1}
         {arriveTangent=-0.0405, interpMode=RCIM_Cubic, leaveTangent=-0.0405, tangentMode=RCTM_User, time=0.4, value=1}
         {arriveTangent=-1.6592, interpMode=RCIM_Cubic, leaveTangent=-1.6592, tangentMode=RCTM_User, time=1, value=0.4}
        ]}}}
     Scale Curve=0.5
    )
    Index 0 Param 3=&FloatFromCurve(
     FloatCurve=DataInterface() {propertyValues {Curve {keys [
         {arriveTangent=1.4725, interpMode=RCIM_Cubic, leaveTangent=1.4725, tangentMode=RCTM_User, time=0, value=0}
         {tangentMode=RCTM_User, time=1, value=1}
        ]}}}
     Scale Curve=&Multiply_Float(
      A=&MaskFloatBySpawnGroup(Masked Float=350, Passthrough Float=&RandomRangeFloat(Minimum=300, Maximum=400), Spawn Group Mask=1)
      B=$Emitter.Global Scale
     )
    )
    Index 0 Param 4=&MaskFloatBySpawnGroup(
     Masked Float=&FloatFromCurve(
      FloatCurve=DataInterface() {propertyValues {Curve {keys [
          {arriveTangent=0.1175, interpMode=RCIM_Cubic, leaveTangent=0.6205, tangentMode=RCTM_Break, time=0.22, value=-2}
          {arriveTangent=6.3488, interpMode=RCIM_Cubic, leaveTangent=6.3488, tangentMode=RCTM_User, time=0.6, value=-0.4172}
          {arriveTangent=36.7876, interpMode=RCIM_Cubic, leaveTangent=14.4841, tangentMode=RCTM_Break, time=1, value=7.3}
         ]}}}
      Scale Curve=0.2
     )
     Passthrough Float=&FloatFromCurve(
      FloatCurve=DataInterface() {propertyValues {Curve {keys [
          {arriveTangent=0.1175, interpMode=RCIM_Cubic, leaveTangent=5.607, tangentMode=RCTM_Break, time=0.1, value=-2.5}
          {arriveTangent=30.0332, interpMode=RCIM_Cubic, leaveTangent=14.4841, tangentMode=RCTM_Break, time=1, value=9}
         ]}}}
      Scale Curve=0.2
     )
     Spawn Group Mask=1
    )
    Index 1 Param 1=&FloatFromCurve(
     FloatCurve=DataInterface() {propertyValues {Curve {keys [
         {arriveTangent=0.5348, interpMode=RCIM_Cubic, leaveTangent=0.5348, tangentMode=RCTM_User, time=0.05, value=0.02}
         {arriveTangent=39.7786, interpMode=RCIM_Cubic, leaveTangent=39.7786, tangentMode=RCTM_User, time=0.4, value=8}
        ]}}}
     Scale Curve=&MaskFloatBySpawnGroup(Masked Float=85, Passthrough Float=&RandomRangeFloat(Minimum=55, Maximum=55), Spawn Group Mask=1)
    )
    Index 1 Param 2=&Divide_Float(
     A=&MaskFloatBySpawnGroup(
      Masked Float=0.01
      Passthrough Float=&RandomRangeFloat(Minimum=0.0089, Maximum=0.015)
      Spawn Group Mask=1
     )
     B=$Emitter.Global Scale
    )
    Index 1 Param 3=&FloatFromCurve(
     FloatCurve=DataInterface() {propertyValues {Curve {keys [{interpMode=RCIM_Cubic, tangentMode=RCTM_User, time=0.4, value=0.5}, <at 0.8 0.1>]}}}
     Scale Curve=5
    )
    Index 1 Param 4=&MaskFloatBySpawnGroup(Masked Float=2, Passthrough Float=1, Spawn Group Mask=1)
    Index 2 Param 1=&FloatFromCurve(FloatCurve=DataInterface() {propertyValues {Curve {keys [
         {arriveTangent=0.5431, interpMode=RCIM_Cubic, leaveTangent=0.5431, tangentMode=RCTM_User, time=0.3, value=0}
         {arriveTangent=2.1449, interpMode=RCIM_Cubic, leaveTangent=2.1449, tangentMode=RCTM_User, time=1, value=1}
        ]}}})
    Index 2 Param 2=&FloatFromCurve(
     FloatCurve=DataInterface() {propertyValues {Curve {keys [
         {interpMode=RCIM_Cubic, tangentMode=RCTM_User, time=0.3, value=0}
         {arriveTangent=2.4333, interpMode=RCIM_Cubic, leaveTangent=2.4333, tangentMode=RCTM_User, time=1, value=1}
        ]}}}
     Scale Curve=&Multiply_Float(
      A=0.1
      B=&FloatFromCurve(
       FloatCurve=DataInterface() {propertyValues {Curve {keys [<at 0 1>,<at 1 1>,<at 10 1>,<at 20 1>,<at 100 0.2>]}}}
       CurveIndex=$Emitter.Global Scale
      )
     )
    )
    Index 2 Param 3=&MaskFloatBySpawnGroup(
     Masked Float=&FloatFromCurve(
      FloatCurve=DataInterface() {propertyValues {Curve {keys [
          {arriveTangent=-6.4028, interpMode=RCIM_Cubic, leaveTangent=-6.4028, tangentMode=RCTM_User, time=0.1, value=1}
          {arriveTangent=-0.0065, interpMode=RCIM_Cubic, leaveTangent=-0.0065, tangentMode=RCTM_User, time=0.4, value=0}
         ]}}}
      Scale Curve=5
     )
     Passthrough Float=&FloatFromCurve(
      FloatCurve=DataInterface() {propertyValues {Curve {keys [
          {arriveTangent=-3.4318, interpMode=RCIM_Cubic, leaveTangent=-3.4318, tangentMode=RCTM_User, time=0.1, value=1}
          {arriveTangent=-0.7527, interpMode=RCIM_Cubic, leaveTangent=-0.7527, tangentMode=RCTM_User, time=0.35, value=0}
         ]}}}
      Scale Curve=5
     )
     Spawn Group Mask=1
    )
    Index 2 Param 4=&FloatFromCurve(
     FloatCurve=DataInterface() {propertyValues {Curve {keys [
         {arriveTangent=-16.6453, interpMode=RCIM_Cubic, leaveTangent=-16.6453, tangentMode=RCTM_User, time=0, value=64}
         {arriveTangent=-87.1792, interpMode=RCIM_Cubic, leaveTangent=-87.1792, tangentMode=RCTM_User, time=0.1, value=24}
         {arriveTangent=-52.5, interpMode=RCIM_Cubic, leaveTangent=-52.5, time=0.2, value=12}
         {arriveTangent=-7.0531, interpMode=RCIM_Cubic, leaveTangent=-7.0531, tangentMode=RCTM_User, time=0.5, value=3}
         {arriveTangent=-0.6909, interpMode=RCIM_Cubic, leaveTangent=-0.6909, tangentMode=RCTM_User, time=1, value=2}
         <at 2 1>
         <at 3 0.3>
         <at 10 0.1>
         <at 20 0.05>
         <at 50 0.02>
         <at 100 0.01>
        ]}}}
     CurveIndex=$Emitter.Global Scale
     Scale Curve=0.6
    )
    Index 3 Param 1=&FloatFromCurve(
     FloatCurve=DataInterface() {propertyValues {Curve {keys [
         {arriveTangent=2.2353, interpMode=RCIM_Cubic, leaveTangent=2.2353, tangentMode=RCTM_User, time=0, value=0.2}
         {arriveTangent=1.4416, interpMode=RCIM_Cubic, leaveTangent=1.4416, tangentMode=RCTM_User, time=0.2, value=0.5}
         {interpMode=RCIM_Cubic, tangentMode=RCTM_User, time=0.8, value=1.25}
        ]}}}
     Scale Curve=-1
    )
    Index 3 Param 2=&FloatFromCurve(
     FloatCurve=DataInterface() {propertyValues {Curve {keys [
         {arriveTangent=5.1563, interpMode=RCIM_Cubic, leaveTangent=5.1563, tangentMode=RCTM_User, time=0, value=0}
         {arriveTangent=1.4416, interpMode=RCIM_Cubic, leaveTangent=1.4416, tangentMode=RCTM_User, time=0.1699, value=0.4}
         {interpMode=RCIM_Cubic, tangentMode=RCTM_User, time=1, value=1.1}
        ]}}}
     Scale Curve=-0.1
    )
    Index 3 Param 3=&FloatFromCurve(FloatCurve=DataInterface() {propertyValues {Curve {keys [
         {arriveTangent=-11.5239, interpMode=RCIM_Cubic, leaveTangent=-11.5239, tangentMode=RCTM_User, time=0.15, value=40}
         {arriveTangent=-88.9092, interpMode=RCIM_Cubic, leaveTangent=-88.9092, tangentMode=RCTM_User, time=0.3, value=25}
         {arriveTangent=-43.7054, interpMode=RCIM_Cubic, leaveTangent=-43.7054, tangentMode=RCTM_User, time=0.6, value=0}
         {interpMode=RCIM_Cubic, tangentMode=RCTM_User, time=0.7, value=0}
         {interpMode=RCIM_Cubic, tangentMode=RCTM_User, time=0.8, value=0}
        ]}}})
    Index 3 Param 4=&FloatFromCurve(
     FloatCurve=DataInterface() {propertyValues {Curve {keys [{tangentMode=RCTM_User, time=0, value=0}, {tangentMode=RCTM_User, time=1, value=1}]}}}
     Scale Curve=-250
    )
    Module.Index0Param0Write=true
    Module.Index0Param1Write=true
    Module.Index0Param2Write=true
    Module.Index0Param3Write=true
    Module.Index1Param0Write=true
    Module.Index1Param1Write=true
    Module.Index1Param2Write=true
    Module.Index1Param3Write=true
    Module.Index2Param0Write=true
    Module.Index2Param1Write=true
    Module.Index2Param2Write=true
    Module.Index2Param3Write=true
    Module.Index3Param0Write=true
    Module.Index3Param1Write=true
    Module.Index3Param2Write=true
    Module.Index3Param3Write=true
    Write Parameter Index 1=true
    Write Parameter Index 2=true
    Write Parameter Index 3=true
   ) {module=DynamicMaterialParameters001}
   &ScaleColor(
    Scale Alpha=&FloatFromCurve(FloatCurve=DataInterface() {propertyValues {Curve {keys [<at 0 1>,<at 1 0>]}}})
    Scale RGB=&MakeVectorFromLinearColorRGB(Color=&MultiplyLinearColor(
      Linear Color=&ColorFromCurve(ColorCurve=DataInterface() {propertyValues {
         AlphaCurve {keys [<at 0.0022 1>,<at 0.994 0>]}
         BlueCurve {keys [<at 0.1888 0.6042>,<at 0.4977 0.4497>]}
         GreenCurve {keys [<at 0.1888 0.6042>,<at 0.4977 0.4497>]}
         RedCurve {keys [<at 0.1888 0.6042>,<at 0.4977 0.4497>]}
        }})
      Scale Factor=&MakeVector4_FromLinearColor(LinearColor=&ColorFromCurve(
        ColorCurve=DataInterface() {propertyValues {
          AlphaCurve {keys [<at 0 1>,<at 0.125 1>,<at 0.25 1>,<at 0.375 1>,<at 0.5 1>,<at 0.625 1>,<at 0.75 1>,<at 0.875 1>,<at 1 1>]}
          BlueCurve {keys [<at 0.125 0.7835>,<at 0.166 0.6172>,<at 0.208 0.5395>,<at 0.25 0.6172>,<at 0.291 0.6939>,<at 0.333 0.7379>,<at 0.375 0.7835>,<at 0.416 0.7991>,<at 0.458 0.8148>,<at 0.5 0.8308>,<at 0.541 0.8148>,<at 0.583 0.7991>,<at 0.625 0.7991>,<at 0.666 0.7529>,<at 0.708 0.7084>,<at 0.75 0.5711>,<at 0.791 0.4735>,<at 0.833 0.3231>,<at 0.8794 0.7835>]}
          GreenCurve {keys [<at 0.125 0.3118>,<at 0.166 0.5333>,<at 0.208 0.5271>,<at 0.25 0.6376>,<at 0.291 0.7529>,<at 0.333 0.7835>,<at 0.375 0.8148>,<at 0.416 0.7991>,<at 0.458 0.8148>,<at 0.5 0.8469>,<at 0.541 0.7991>,<at 0.583 0.7991>,<at 0.625 0.8148>,<at 0.666 0.7991>,<at 0.708 0.7529>,<at 0.75 0.5711>,<at 0.791 0.5029>,<at 0.833 0.5029>,<at 0.8794 0.3118>]}
          RedCurve {keys [<at 0.125 0.1252>,<at 0.166 0.7084>,<at 0.208 0.6867>,<at 0.25 0.8148>,<at 0.291 0.8879>,<at 0.333 0.8963>,<at 0.375 0.9131>,<at 0.416 0.8714>,<at 0.458 0.8714>,<at 0.5 0.9131>,<at 0.541 0.8632>,<at 0.583 0.8714>,<at 0.625 0.9131>,<at 0.666 0.8963>,<at 0.708 0.8714>,<at 0.75 0.7529>,<at 0.791 0.7529>,<at 0.833 0.8469>,<at 0.8794 0.1252>]}
         }}
        CurveIndex=Linked() {linkedVariable {name=NPC.FortniteNPC.TimeOfDay_0-1, type=NiagaraFloat}}
       ))
     ))
    ScaleA=true
    ScaleRGB=true
   )
   &CurlNoiseForce(
    Noise Frequency=&Divide_Float(A=4, B=$Emitter.Global Scale)
    Noise Quality / Cost=::NewEnumerator0
    Noise Strength=&Multiply_Float(A=&MaskFloatBySpawnGroup(Masked Float=0, Passthrough Float=1000, Spawn Group Mask=1), B=$Emitter.Global Scale)
    Pan Noise=false
   )
   &PointAttractionForce(
    Attraction Radius=500
    Attraction Strength=&FloatFromCurve(FloatCurve=DataInterface() {propertyValues {Curve {
        defaultValue=-500
        keys [
         {arriveTangent=-119.0229, interpMode=RCIM_Cubic, leaveTangent=-119.0229, tangentMode=RCTM_User, time=0, value=-10}
         {arriveTangent=1163.9872, interpMode=RCIM_Cubic, leaveTangent=1163.9872, tangentMode=RCTM_User, time=0.5, value=400}
        ]
       }}})
    Kill Radius:NiagaraFloat=&FloatFromCurve(FloatCurve=DataInterface() {propertyValues {Curve {
        defaultValue=-500
        keys [
         {arriveTangent=2.4671, interpMode=RCIM_Cubic, leaveTangent=2.4671, tangentMode=RCTM_User, time=0, value=0}
         {arriveTangent=18.5105, interpMode=RCIM_Cubic, leaveTangent=18.5105, tangentMode=RCTM_User, time=0.5, value=15}
        ]
       }}})
    Kill Within Radius=true
    Module.Use Falloff=true
    UseAttractorOffset=false
   )
   &VortexForce(UseOriginOffset=false, Vortex Force Amount=500)
   &SolveForcesAndVelocity(Clamp Velocity=false, Limit Acceleration=false)
   &ScaleMeshSize(Scale Factor=&VectorFromFloat(Value=&FloatFromCurve(
      FloatCurve=DataInterface() {propertyValues {Curve {
         defaultValue=1
         keys [
          {arriveTangent=9.0725, interpMode=RCIM_Cubic, leaveTangent=9.0725, tangentMode=RCTM_User, time=0, value=0}
          {arriveTangent=-2.5249, interpMode=RCIM_Cubic, leaveTangent=-2.5249, tangentMode=RCTM_User, time=0.1408, value=0.9847}
          {arriveTangent=-0.0405, interpMode=RCIM_Cubic, leaveTangent=-0.0405, tangentMode=RCTM_User, time=1, value=0}
         ]
        }}}
      Scale Curve=0.8
     )))
  ]
  rendererClasses [&NiagaraMeshRendererProperties]
  renderers [{rendererClass=&NiagaraMeshRendererProperties, rendererIndex=0}]
 }
 SmokeyTrails {
  emitterUpdateScript [
   &EmitterState(Life Cycle Mode=::NewEnumerator1, Loop Behavior=::NewEnumerator1, Loop Duration=0.15, UseLoopDelay=false)
   &SpawnRate(SpawnRate=70, Use Spawn Probability=false)
  ]
  particleSpawnScript [
   &InitializeParticle(
    Lifetime Max=0.6
    Lifetime Min=0.5
    Lifetime Mode=::NewEnumerator1
    Mesh Scale Max=<1.1 1.1 1.1>
    Mesh Scale Mode=::NewEnumerator2
    UsePositionOffset=false
    Write Material Random=true
   )
   &InitialMeshOrientation(
    Mesh Orientation Mode=::NewEnumerator2
    Rotation=&RandomRangeVector(Minimum=<0 0 0>, Maximum=<1 1 1>)
    Use Rotation Vector=true
   )
  ]
  particleUpdateScript [
   &ParticleState()
   &GravityForce(Gravity=<0 0 -0.01>)
   &ScaleColor(
    Scale Alpha=&FloatFromCurve(FloatCurve=DataInterface() {propertyValues {Curve {keys [
         {arriveTangent=200.5769, interpMode=RCIM_Cubic, leaveTangent=200.5752, tangentMode=RCTM_User, time=0, value=0}
         {arriveTangent=0.0668, interpMode=RCIM_Cubic, leaveTangent=0.0668, tangentMode=RCTM_User, time=0.01, value=1}
         <at 1 0>
        ]}}})
    Scale RGB=&VectorFromCurve(VectorCurve=DataInterface() {propertyValues {
       bUseLUT=false
       XCurve {keys [
         <at -0.0009 0.9906 RCIM_Cubic>
         {arriveTangent=-1.9223, interpMode=RCIM_Cubic, leaveTangent=-1.0067, tangentMode=RCTM_Break, time=0.3779, value=0.4}
         <at 1 0 RCIM_Cubic>
        ]}
       YCurve {keys [
         {arriveTangent=-3.6896, interpMode=RCIM_Cubic, leaveTangent=-3.6896, tangentMode=RCTM_User, time=-0.0011, value=1.0067}
         {arriveTangent=-0.3437, interpMode=RCIM_Cubic, leaveTangent=-0.3437, tangentMode=RCTM_User, time=0.3551, value=0.2247}
         <at 1 0 RCIM_Cubic>
        ]}
       ZCurve {keys [
         <at -0.0011 0 RCIM_Cubic>
         {arriveTangent=-0.223, interpMode=RCIM_Cubic, leaveTangent=-0.223, tangentMode=RCTM_User, time=0.9965, value=0.7415}
        ]}
      }})
    ScaleA=true
    ScaleRGB=true
   )
   &ScaleMeshSize(Scale Factor=&VectorFromFloat(Value=&FloatFromCurve(
      FloatCurve=DataInterface() {propertyValues {Curve {
         defaultValue=0
         keys [
          {arriveTangent=-0.0405, interpMode=RCIM_Cubic, leaveTangent=-0.0405, tangentMode=RCTM_User, time=0, value=0}
          {arriveTangent=9.0725, interpMode=RCIM_Cubic, leaveTangent=9.0725, tangentMode=RCTM_User, time=0, value=0}
          {arriveTangent=-2.7479, interpMode=RCIM_Cubic, leaveTangent=-2.7479, tangentMode=RCTM_User, time=0.1408, value=0.9847}
          <at 0.6 0 RCIM_Cubic>
         ]
        }}}
      Scale Curve=&FN_FrontEndFloatModifier(Gameplay=1, Front End=0.8)
     )))
   &DynamicMaterialParameters(
    Index 0 Param 1=&FloatFromCurve(FloatCurve=DataInterface() {propertyValues {Curve {keys [
         {arriveTangent=3.1529, interpMode=RCIM_Cubic, leaveTangent=3.1529, tangentMode=RCTM_User, time=0, value=0.5}
         {arriveTangent=0.4946, interpMode=RCIM_Cubic, leaveTangent=0.4946, tangentMode=RCTM_User, time=0.1931, value=0.9256}
         <at 1.001 1.1573>
        ]}}})
    Index 0 Param 2=&FloatFromCurve(FloatCurve=DataInterface() {propertyValues {Curve {keys [<at 0.505 0.9841>,<at 0.7325 0.0093>,<at 1.0011 0.002>]}}})
    Index 0 Param 3=&FloatFromCurve(
     FloatCurve=DataInterface() {propertyValues {Curve {
        defaultValue=1.0076
        keys [
         {arriveTangent=9.0725, interpMode=RCIM_Cubic, leaveTangent=9.0725, tangentMode=RCTM_User, time=0, value=0}
         {arriveTangent=-2.5249, interpMode=RCIM_Cubic, leaveTangent=-2.5249, tangentMode=RCTM_User, time=0.1408, value=0.9847}
         {arriveTangent=-0.0405, interpMode=RCIM_Cubic, leaveTangent=-0.0405, tangentMode=RCTM_User, time=1, value=0.25}
        ]
       }}}
     Scale Curve=0.25
    )
    Module.Param0WriteEnabled=true
    Module.Param1WriteEnabled=true
    Module.Param2WriteEnabled=true
    Module.Param3WriteEnabled=true
   )
   &SolveForcesAndVelocity(Clamp Velocity=false, Limit Acceleration=false)
  ]
  rendererClasses [&NiagaraMeshRendererProperties]
  renderers [{rendererClass=&NiagaraMeshRendererProperties, rendererIndex=0}]
 }
 Sparky_Trails {
  emitterSpawnScript [/SetVariables(Emitter.RandomVectorRotation:NiagaraFloat=&RandomRangeFloat(Minimum=0, Maximum=360)) {isSetParameters=true}]
  emitterUpdateScript [
   &EmitterState(Life Cycle Mode=::NewEnumerator1, Loop Behavior=::NewEnumerator1, Loop Duration=1, UseLoopDelay=false)
   &SpawnBurst_Instantaneous(Spawn Count=100, Use Spawn Probability=false, UseLoopCountLimit=false)
  ]
  particleSpawnScript [
   &InitializeParticle(
    Color Mode=::NewEnumerator0
    Lifetime Max=0.8
    Lifetime Min=0.5
    Lifetime Mode=::NewEnumerator1
    Position Offset=&Multiply_VectorByFloat(
     Vector=/Niagara/DynamicInputs/Vectors/RotateVector(
      Vector To Rotate:Vector3f=&SelectVectorFromArray(
       Direct Array Index=&ReturnParticleIDAttributes()
       Direct Array Mode=::NewEnumerator1
       Vector Selection Array=DataInterface() {propertyValues {FloatData [<0 0.5 0>, <0.5405 -1.2953 1.0181>, <0.8 0.964 0.5>, <-1 0 0.6937>]}}
      )
      Yaw:NiagaraFloat=$Emitter.RandomVectorRotation
      Pitch:NiagaraFloat=0
      Roll:NiagaraFloat=0
      Rotation Rate:NiagaraFloat=&Add_Float(A=1, B=0)
     )
     Float=5
    )
    Sprite Size Max=<3 5>
    Sprite Size Min=<0.5 2>
    Sprite Size Mode=::NewEnumerator2
    UsePositionOffset=true
    Write Material Random=true
   ) {module=InitializeParticle001}
   &AddVelocity(
    Velocity=&RandomRangeVector(Minimum=<0 0 5>, Maximum=<0 0 5>)
    Velocity Speed Scale=&RandomRangeFloat(Minimum=8, Maximum=15)
   )
   &AddVelocity(
    Cone Angle=32.7284
    Cone Axis=&RotateVector(
     Vector To Rotate:Vector3f=&SelectVectorFromArray(
      Direct Array Index=&ReturnParticleIDAttributes()
      Direct Array Mode=::NewEnumerator1
      Vector Selection Array=DataInterface() {propertyValues {FloatData [<0 0.5 0>, <0.5405 -1.2953 1.0181>, <0.8 0.964 0.5>, <-1 0 0.6937>]}}
     )
     Yaw:NiagaraFloat=$Emitter.RandomVectorRotation
     Pitch:NiagaraFloat=0
     Roll:NiagaraFloat=0
     Rotation Rate:NiagaraFloat=&Add_Float(A=1, B=0)
    )
    Module.UseConeFalloff=false
    Module.UseVelDistribution=false
    Override Seed=false
    Velocity Mode=::NewEnumerator2
    Velocity Speed=&Multiply_Float(
     A=&RandomRangeFloat(Minimum=0, Maximum=683.378)
     B=&FN_FrontEndFloatModifier(Gameplay=1, Front End=1)
    )
   ) {module=AddVelocity001}
  ]
  particleUpdateScript [
   &ParticleState()
   &AccelerationForce(Acceleration=&RandomRangeVector(Minimum=<0 0 -200>, Maximum=<0 0 -400>))
   &Drag(Drag=&RandomRangeFloat(Minimum=4, Maximum=7), Use Linear Drag=true, Use Rotational Drag=true)
   &PointAttractionForce(
    Attraction Radius=500
    Attraction Strength=&FloatFromCurve(FloatCurve=DataInterface() {propertyValues {Curve {
        defaultValue=-500
        keys [
         {arriveTangent=-119.0229, interpMode=RCIM_Cubic, leaveTangent=-119.0229, tangentMode=RCTM_User, time=0, value=-10}
         {arriveTangent=1163.9872, interpMode=RCIM_Cubic, leaveTangent=1163.9872, tangentMode=RCTM_User, time=0.5, value=400}
        ]
       }}})
    Kill Radius:NiagaraFloat=&FloatFromCurve(FloatCurve=DataInterface() {propertyValues {Curve {
        defaultValue=-500
        keys [
         {arriveTangent=2.4671, interpMode=RCIM_Cubic, leaveTangent=2.4671, tangentMode=RCTM_User, time=0, value=0}
         {arriveTangent=18.5105, interpMode=RCIM_Cubic, leaveTangent=18.5105, tangentMode=RCTM_User, time=0.5, value=15}
        ]
       }}})
    Kill Within Radius=true
    Module.Use Falloff=true
    UseAttractorOffset=false
   )
   &VortexForce(UseOriginOffset=false, Vortex Force Amount=600)
   &SolveForcesAndVelocity(Clamp Velocity=false, Limit Acceleration=false)
   &ScaleSpriteSizeBySpeed(Max Scale Factor=<1 5>, Min Scale Factor=<0.65 0.65>, Velocity Threshold=500)
   &ScaleSpriteSize(Scale Factor=&Multiply_Vector2D_ByFloat(
     Float=1
     Vector2D=&Vector2DFromCurve(
      Vector2Curve=DataInterface() {propertyValues {XCurve {keys [<at 0 0.5>,<at 0.2 1>,<at 1 1>]}, YCurve {keys [<at 0 3>,<at 0.2 1>,<at 1 0>]}}}
      ScaleCurve=<1 1>
     )
    ))
   &ScaleColor(
    Scale Alpha=&FN_FrontEndFloatModifier(Gameplay=1, Front End=0.25)
    Scale RGB=&MakeVectorFromLinearColorRGB(Color=&ColorFromCurve(ColorCurve=DataInterface() {propertyValues {
        AlphaCurve {keys [<at 0 1 RCIM_Cubic>]}
        BlueCurve {keys [<at 0 0 RCIM_Cubic>,<at 1 0 RCIM_Cubic>]}
        GreenCurve {keys [<at 0 9.1667 RCIM_Cubic>,<at 1 0 RCIM_Cubic>]}
        RedCurve {keys [<at 0 50 RCIM_Cubic>,<at 1 1 RCIM_Cubic>]}
       }}))
    ScaleA=true
    ScaleRGB=true
   )
  ]
  rendererClasses [&NiagaraSpriteRendererProperties]
  renderers [{rendererClass=&NiagaraSpriteRendererProperties}]
 }
 Streaks {
  emitterUpdateScript [
   &EmitterState(Life Cycle Mode=::NewEnumerator1, Loop Behavior=::NewEnumerator1, Loop Duration=0.4, UseLoopDelay=false)
   &SpawnBurst_Instantaneous(Spawn Count=10, Use Spawn Probability=false)
  ]
  particleSpawnScript [
   &InitializeParticle(
    Color=<rgba 10 3.77 0 1>
    Lifetime=&UniformRangedFloat(Minimum=0.15, Maximum=0.25)
    Mass Mode=::NewEnumerator1
    Sprite Size=&Multiply_Vector2D_ByFloat(Float=1, Vector2D=&UniformRangedVector2D(Minimum=<1 20>, Maximum=<5 40>))
    Sprite Size Mode=::NewEnumerator4
    UsePositionOffset=true
    Write Material Random=true
   ) {module=InitializeParticle001}
   /Niagara/Modules/Spawn/Location/SphereLocation(
    MaskSpawnGroup=false
    Module.Override Randomness=false
    Sphere Radius=&Multiply_Float(A=5, B=&FN_FrontEndFloatModifier(Gameplay=1, Front End=$System.FrontEndScale))
    Surface Only=true
   )
   &AddVelocityFromPoint(
    Module.Use Falloff Distance=false
    Velocity Strength=&Multiply_Float(
     A=&UniformRangedFloat(Minimum=100, Maximum=200)
     B=&FN_FrontEndFloatModifier(Gameplay=1, Front End=$System.FrontEndScale)
    )
   )
  ]
  particleUpdateScript [
   &ParticleState()
   &SolveForcesAndVelocity(Clamp Velocity=false, Limit Acceleration=false)
   &ScaleSpriteSize(Scale Factor=&Multiply_Vector2D_ByFloat(
     Float=&FloatFromCurve(FloatCurve=DataInterface() {propertyValues {Curve {
         defaultValue=1
         keys [
          {arriveTangent=9.0725, interpMode=RCIM_Cubic, leaveTangent=9.0725, tangentMode=RCTM_User, time=0, value=0}
          {arriveTangent=-2.5249, interpMode=RCIM_Cubic, leaveTangent=-2.5249, tangentMode=RCTM_User, time=0.1408, value=0.9847}
          {arriveTangent=-0.0405, interpMode=RCIM_Cubic, leaveTangent=-0.0405, tangentMode=RCTM_User, time=1, value=0}
         ]
        }}})
     Vector2D=&Vector2DFromCurve(
      Vector2Curve=DataInterface() {propertyValues {XCurve {keys [<at 0 0.75 RCIM_Cubic>,<at 1 1 RCIM_Cubic>]}, YCurve {keys [<at 0 1 RCIM_Cubic>,<at 1 2 RCIM_Cubic>]}}}
      ScaleCurve=<1 1>
     )
    ))
   &ScaleColor(
    Scale Alpha=&FloatFromCurve(FloatCurve=DataInterface() {propertyValues {Curve {keys [<at 0 1 RCIM_Cubic>,<at 1 0>]}}})
    Scale RGB=&VectorFromFloat(Value=&FloatFromCurve(FloatCurve=DataInterface() {propertyValues {Curve {keys [<at 0 10 RCIM_Cubic>,<at 1 1 RCIM_Cubic>]}}}))
    ScaleA=true
    ScaleRGB=true
   )
  ]
  rendererClasses [&NiagaraSpriteRendererProperties]
  renderers [{rendererClass=&NiagaraSpriteRendererProperties}]
 }
 Waves {
  emitterUpdateScript [
   &EmitterState(Life Cycle Mode=::NewEnumerator1, Loop Behavior=::NewEnumerator1, Loop Duration=1, UseLoopDelay=true)
   &SpawnBurst_Instantaneous(Spawn Time=0.05, Use Spawn Probability=false, UseLoopCountLimit=false)
   &SpawnBurst_Instantaneous(Spawn Group=1, Use Spawn Probability=false, UseLoopCountLimit=false) {module=SpawnBurst_Instantaneous001}
  ]
  particleSpawnScript [
   /SetVariables(Particles.VisibilityTag=&SpawnGroup()) {isSetParameters=true}
   &InitializeParticle(
    Color Mode=::NewEnumerator0
    Lifetime=&MergeHelper_FloatSwitch(0=0.25, 1=1, 2=?, 3=?, Index=&SpawnGroup(), Num=::NewEnumerator0)
    Mesh Scale Mode=::NewEnumerator3
    Mesh Uniform Scale=&MergeHelper_FloatSwitch(0=1, 1=0.8, 2=?, 3=?, Index=&SpawnGroup(), Num=::NewEnumerator0)
    UsePositionOffset=false
    Write Material Random=true
   )
   &InitialMeshOrientation(Orientation Axis=<0 0 1>, Orientation Vector=<0 0 1>, Rotation=<0 0.25 0>, Use Orientation Vector=true, Use Rotation Vector=true)
   &InitialMeshOrientation(
    Orientation Axis=<0 0 0>
    Orientation Vector=<0 0 0>
    Rotation=&RandomRangeVector(Minimum=<0 0 0>, Maximum=<1 0 0>)
    Use Orientation Vector=false
    Use Rotation Vector=true
   ) {module=InitialMeshOrientation001}
   &CameraOffset(Camera Offset Amount=30)
  ]
  particleUpdateScript [
   &ParticleState()
   &ScaleColor(
    Color Scale:LinearColor=&MergeHelper_ColorSwitch(
     0=&ColorFromCurve(ColorCurve=DataInterface() {propertyValues {
        AlphaCurve {keys [<at 0 1>,<at 0.3576 1>,<at 1 0>]}
        BlueCurve {keys [<at 0 1>,<at 0.3 1>,<at 0.8 1>]}
        GreenCurve {keys [<at 0 0.5489>,<at 0.3 0.3272>,<at 0.8 0.0414>]}
        RedCurve {keys [<at 0 0.0978>,<at 0.3 0.1029>,<at 0.8 0.2012>]}
       }})
     1=&ColorFromCurve(ColorCurve=DataInterface() {propertyValues {
        AlphaCurve {keys [<at 0 1>,<at 0.3576 1>,<at 1 0>]}
        BlueCurve {keys [<at 0 1>,<at 0.1084 1>,<at 0.25 1>,<at 1 1>]}
        GreenCurve {keys [<at 0 0>,<at 0.1084 0.5489>,<at 0.25 0.1777>,<at 1 0.0414>]}
        RedCurve {keys [<at 0 0.1667>,<at 0.1084 0.0978>,<at 0.25 0.1029>,<at 1 0.2012>]}
       }})
     2=?
     3=?
     Index=&SpawnGroup()
     Num=::NewEnumerator0
    )
    Scale Mode=3
   )
   &DynamicMaterialParameters(
    Index 0 Param 1=&MergeHelper_FloatSwitch(
     0=&FloatFromCurve(FloatCurve=DataInterface() {propertyValues {Curve {keys [{arriveTangent=3.726, interpMode=RCIM_Cubic, leaveTangent=3.726, tangentMode=RCTM_User, time=0, value=-0.3}, <at 1 1>]}}})
     1=&FloatFromCurve(FloatCurve=DataInterface() {propertyValues {Curve {keys [{arriveTangent=-4.0141, interpMode=RCIM_Cubic, leaveTangent=-4.0141, tangentMode=RCTM_User, time=0, value=0.6}, <at 1 -0.7>]}}})
     2=?
     3=?
     Index=&SpawnGroup()
     Num=::NewEnumerator0
    )
    Index 0 Param 2=&MergeHelper_FloatSwitch(
     0=&FloatFromCurve(FloatCurve=DataInterface() {propertyValues {Curve {keys [{arriveTangent=-0.5818, interpMode=RCIM_Cubic, leaveTangent=-0.5818, tangentMode=RCTM_User, time=0, value=1}, <at 1 0.8>]}}})
     1=&FloatFromCurve(FloatCurve=DataInterface() {propertyValues {Curve {keys [{arriveTangent=-0.5818, interpMode=RCIM_Cubic, leaveTangent=-0.5818, tangentMode=RCTM_User, time=0, value=1}, <at 1 0.9>]}}})
     2=?
     3=?
     Index=&SpawnGroup()
     Num=::NewEnumerator0
    )
    Index 0 Param 3=&MergeHelper_FloatSwitch(
     0=&FloatFromCurve(FloatCurve=DataInterface() {propertyValues {Curve {keys [<at 0 0>,<at 0.15 1>,<at 1 1.2>]}}})
     1=&FloatFromCurve(FloatCurve=DataInterface() {propertyValues {Curve {keys [<at 0 0>,<at 0.15 1>,<at 1 1.2>]}}})
     2=?
     3=?
     Index=&SpawnGroup()
     Num=::NewEnumerator0
    )
    Module.Index0Float0Write=true
    Module.Index0Float1Write=true
    Module.Index0Float2Write=true
    Module.Index0Float3Write=false
   )
   &ScaleColor(
    Scale Alpha=&MergeHelper_FloatSwitch(
     0=&Multiply_Float(A=2, B=&FNTimeOfDayScalar(Night Scale=0.7, Morning Scale=0.5, Evening Scale=0.5))
     1=&Multiply_Float(A=1, B=&FNTimeOfDayScalar(Night Scale=0.7, Morning Scale=0.5, Evening Scale=0.5))
     2=?
     3=?
     Index=&SpawnGroup()
     Num=::NewEnumerator0
    )
    Scale RGB=&VectorFromFloat(Value=&MergeHelper_FloatSwitch(
      0=&FloatFromCurve(FloatCurve=DataInterface() {propertyValues {Curve {keys [<at 0 70>,<at 0.15 3>,<at 1 1>]}}})
      1=&FloatFromCurve(FloatCurve=DataInterface() {propertyValues {Curve {keys [<at 0 10>,<at 0.15 3>,<at 1 1>]}}})
      2=?
      3=?
      Index=&SpawnGroup()
      Num=::NewEnumerator0
     ))
    ScaleA=true
    ScaleRGB=true
   ) {module=ScaleColor001}
   &ScaleMeshSize(Scale Factor=&VectorFromFloat(Value=&MergeHelper_FloatSwitch(
      0=&FloatFromCurve(FloatCurve=DataInterface() {propertyValues {Curve {keys [<at 0 0.5>,<at 0.15 1>,<at 1 1.2>]}}})
      1=&FloatFromCurve(FloatCurve=DataInterface() {propertyValues {Curve {keys [<at 0 0>, {arriveTangent=-6.1514, interpMode=RCIM_Cubic, leaveTangent=-6.1514, tangentMode=RCTM_User, time=0.1, value=2}, <at 1 0>]}}})
      2=?
      3=?
      Index=&SpawnGroup()
      Num=::NewEnumerator0
     )))
  ]
  rendererClasses [&NiagaraMeshRendererProperties]
  renderers [
   {rendererClass=&NiagaraMeshRendererProperties, rendererIndex=0}
   {rendererClass=&NiagaraMeshRendererProperties, rendererIndex=1}
  ]
 }
}
summary {
 systemName=NS_Emote_HeatShineTorn_Implosion_01
 emitters [
  {emitterName=Streaks, rendererClasses [&NiagaraSpriteRendererProperties]}
  {emitterName=Sparky_Trails, rendererClasses [&NiagaraSpriteRendererProperties]}
  {emitterName=LWE_Refract, rendererClasses []}
  {emitterName=Smoke, rendererClasses [&NiagaraMeshRendererProperties]}
  {emitterName=SmokeyTrails, rendererClasses [&NiagaraMeshRendererProperties]}
  {
   emitterName=Refraction_Flare
   rendererClasses [
    &NiagaraMeshRendererProperties
    &NiagaraSpriteRendererProperties
    &NiagaraLightRendererProperties
   ]
  }
  {emitterName=Waves, rendererClasses [&NiagaraMeshRendererProperties]}
 ]
 userVariables [
  User.bIsFrontEnd {type=NiagaraBool}
  User.bIsFrontEndPreview {type=NiagaraBool}
 ]
}
